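# Control Rig for the Maroon ensemble coat on a medium male body, posing the coat's dyn_skirt bones from leg motion.
# It runs a shared DeformRig function, then spherical pose readers on thighs and calves store weights for up, back, side and inward poses.
# Those weights drive additive transform offsets and blended parent constraints that tie skirt bones to thighs, calves and each other.
# An IsRBANEnabled flag disables calf constraints on lower skirt bones, and a jumping value scales down one right-side bend.
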ControlRigBlueprint /BRCosmetics/Characters/Player/Male/Medium/Bodies/M_MED_Ensemble_Maroon/Meshes/M_MED_Ensemble_Maroon_coat_CtrlRig
graphs [
 ContainedGraph {nodes [
   n1 = RigVMFunctionEntryNode()
   RigUnit_GetTransform = &RigUnit_GetTransform(Space=LocalSpace, bInitial=false, Item=@n1.Joint)
   n2 = &RigVMFunction_MathQuaternionToEuler(RotationOrder=ZYX, Value=@RigUnit_GetTransform.Transform.Rotation)
   > RigUnit_SetTransform = &RigUnit_SetTransform(
    Space=LocalSpace
    Value=(Rotation=(X=0.0132,Y=0.4099,Z=-0.0222,W=0.9118),Translation=(X=23.6045,Y=-6.37,Z=140.1115),Scale3D=(X=1,Y=1,Z=1))
    bInitial=false
    Item=@n1.Joint
    Weight=@n1.Weight
    Value.Translation.Z=&RigVMFunction_MathFloatAdd(B=@n1.Translate Z, A=@RigUnit_GetTransform.Transform.Translation.Z) {template="Add::Execute(in A,in B,out Result)"}.Result
    Value.Scale3D.Z=&RigVMFunction_MathFloatAdd(B=@n1.Scale Z, A=@RigUnit_GetTransform.Transform.Scale3D.Z) {template="Add::Execute(in A,in B,out Result)"}.Result
    Value.Translation.Y=&RigVMFunction_MathFloatAdd(B=@n1.Translate Y, A=@RigUnit_GetTransform.Transform.Translation.Y) {template="Add::Execute(in A,in B,out Result)"}.Result
    Value.Scale3D.Y=&RigVMFunction_MathFloatAdd(B=@n1.Scale Y, A=@RigUnit_GetTransform.Transform.Scale3D.Y) {template="Add::Execute(in A,in B,out Result)"}.Result
    Value.Translation.X=&RigVMFunction_MathFloatAdd(B=@n1.Translate X, A=@RigUnit_GetTransform.Transform.Translation.X) {template="Add::Execute(in A,in B,out Result)"}.Result
    Value.Scale3D.X=&RigVMFunction_MathFloatAdd(B=@n1.Scale X, A=@RigUnit_GetTransform.Transform.Scale3D.X) {template="Add::Execute(in A,in B,out Result)"}.Result
    Value.Rotation=&RigVMFunction_MathQuaternionFromEuler(
     Euler=(X=0,Y=0,Z=0)
     Euler.X=&RigVMFunction_MathFloatAdd(B=@n1.Rotate X, A=@n2.Result.X) {template="Add::Execute(in A,in B,out Result)"}.Result
     Euler.Y=&RigVMFunction_MathFloatAdd(B=@n1.Rotate Y, A=@n2.Result.Y) {template="Add::Execute(in A,in B,out Result)"}.Result
     Euler.Z=&RigVMFunction_MathFloatAdd(B=@n1.Rotate Z, A=@n2.Result.Z) {template="Add::Execute(in A,in B,out Result)"}.Result
    ).Result
   )
   > RigVMFunctionReturnNode()
  ]}
 RigVMFunctionLibrary {nodes [Set Transforms for Bone = RigVMCollapseNode(
    ExecuteContext=()
    Joint=()
    Rotate X=0
    Rotate Y=0
    Rotate Z=0
    Scale X=0
    Scale Y=0
    Scale Z=0
    Translate X=0
    Translate Y=0
    Translate Z=0
    Weight=1
   )]}
 RigVMFunction_Sequence_1_ContainedGraph {nodes [
   n1 = RigVMFunctionEntryNode()
   RigVMFunction_Sequence = &RigVMFunction_Sequence(ExecuteContext=@n1.ExecuteContext)
   n2 = &RigVMFunction_Sequence(ExecuteContext=@RigVMFunction_Sequence.B)
   n3 = &RigVMFunction_Sequence(ExecuteContext=@n2.B)
   n4 = &RigVMFunction_Sequence(ExecuteContext=@n3.B)
   n5 = &RigVMFunction_Sequence(ExecuteContext=@n4.B)
   n6 = &RigVMFunction_Sequence(ExecuteContext=@n5.B)
   RigVMFunctionReturnNode(A=@RigVMFunction_Sequence.A, B=@n2.A, C=@n3.A, D=@n4.A, E=@n5.A, F=@n6.A, G=@n6.B)
  ]}
 RigVMModel {nodes [
   RigVMFunction_MathBoolToFloat = &RigVMFunction_MathBoolToFloat(Value=RigVMVariableNode() {variable=IsonGrindRail}.Value)
   &RigUnit_ParentConstraint(
    AdvancedSettings=(InterpolationType=Average,RotationOrderForFilter=XZY)
    Child=(Type=Bone,Name="dyn_skirt_l_02")
    ExecutePin=()
    Parents=((Item=(Type=Bone,Name="thigh_l"),Weight=1))
    Weight=@RigVMFunction_MathBoolToFloat.Result
   )
   > &RigUnit_ParentConstraint(
    AdvancedSettings=(InterpolationType=Average,RotationOrderForFilter=XZY)
    Child=(Type=Bone,Name="dyn_skirt_front_l_03")
    Parents=((Item=(Type=Bone,Name="thigh_l"),Weight=1))
    Weight=@RigVMFunction_MathBoolToFloat.Result
   )
   > &RigUnit_ParentConstraint(
    AdvancedSettings=(InterpolationType=Average,RotationOrderForFilter=XZY)
    Child=(Type=Bone,Name="dyn_skirt_front_l_04")
    Parents=((Item=(Type=Bone,Name="thigh_l"),Weight=1))
    Weight=1
    Parents.0.Weight=@RigVMFunction_MathBoolToFloat.Result
   )
   > n1 = &RigUnit_ParentConstraint(
    AdvancedSettings=(InterpolationType=Average,RotationOrderForFilter=XZY)
    Child=(Type=Bone,Name="dyn_skirt_front_l_05")
    Parents=((Item=(Type=Bone,Name="calf_l"),Weight=1))
    Weight=@RigVMFunction_MathBoolToFloat.Result
   )
   n2 = /Script/ControlRig.RigUnit_SequenceExecution(ExecuteContext=@n1.ExecutePin) {is=struct RigVMUnitNode}
   n3 = &RigUnit_SphericalPoseReader(
    ActiveRegionScaleFactors=(PositiveWidth=3.5,NegativeWidth=1,PositiveHeight=3.5,NegativeHeight=4)
    ActiveRegionSize=0.1
    Debug=(bDrawDebug=true,bDraw2D=false,bDrawLocalAxes=false,DebugScale=25,DebugSegments=20,DebugThickness=0.25)
    DriverAxis=(X=-1,Y=0,Z=0)
    DriverItem=(Type=Bone,Name="thigh_l")
    FalloffRegionScaleFactors=(PositiveWidth=1,NegativeWidth=0.9,PositiveHeight=1,NegativeHeight=1)
    FalloffSize=0.45
    FlipHeightScaling=true
    FlipWidthScaling=true
    OptionalParentItem=(Type=Bone,Name="None")
    RotationOffset=(X=90,Y=180,Z=0)
    ExecutePin=@n2.A
   )
   RigVMFunction_MathBoolNot = &RigVMFunction_MathBoolNot(Value=RigVMVariableNode() {variable=IsonGrindRail}.Value)
   > &RigUnit_ModifyTransforms(
    ItemToModify="((Item=(Type=Bone,Name=\"dyn_skirt_front_c_MSTR\"),Transform=(Rotation=(X=0,Y=0,Z=0,W=1),Translation=(X=-4,Y=0,Z=8),Scale3D=(X=1,Y=1,Z=1))),(Item=(Type=Bone,Name=\"dyn_skirtH_front_l_MSTR\"),Transform=(Rotation=(X=0,Y=0,Z=0,W=1),Translation=(X=-7,Y=0,Z=8),Scal ...(truncated)"
    ItemToModify.0.Transform.Rotation=&RigVMFunction_MathQuaternionFromEuler(Euler=(X=0,Y=0,Z=-30)).Result
    ItemToModify.2.Transform.Rotation=&RigVMFunction_MathQuaternionFromEuler(Euler=(X=0,Y=15,Z=-15)).Result
    ItemToModify.1.Transform.Rotation=&RigVMFunction_MathQuaternionFromEuler(Euler=(X=0,Y=-10,Z=-25)).Result
    Weight=RigVMDispatchNode(Values=(0,0), Index=&RigVMFunction_MathBoolToInteger(Value=@RigVMFunction_MathBoolNot.Result).Result, Values.0=@n3.OutputParam) {template="DISPATCH_RigVMDispatch_SelectInt32(in Index,in Values,out Result)"}.Result
   )
   > &RigUnit_ModifyTransforms(
    ItemToModify=((Item=(Type=Bone,Name="dyn_skirt_l_MSTR"),Transform=(Rotation=(X=0,Y=0,Z=0,W=1),Translation=(X=-7,Y=0,Z=5),Scale3D=(X=1,Y=1,Z=1))))
    ItemToModify.0.Transform.Rotation=&RigVMFunction_MathQuaternionFromEuler(Euler=(X=0,Y=-15,Z=-15)).Result
    Weight=@n3.OutputParam
   )
   > &RigUnit_ModifyTransforms(
    ItemToModify=((Item=(Type=Bone,Name="dyn_skirt_bk_l_MSTR"),Transform=(Rotation=(X=0,Y=0,Z=0,W=1),Translation=(X=0,Y=0,Z=0),Scale3D=(X=1,Y=1,Z=1))))
    ItemToModify.0.Transform.Rotation=&RigVMFunction_MathQuaternionFromEuler(Euler=(X=0,Y=-15,Z=0)).Result
    Weight=@n3.OutputParam
   )
   n4 = &RigUnit_SphericalPoseReader(
    ActiveRegionScaleFactors=(PositiveWidth=3.5,NegativeWidth=1,PositiveHeight=3.5,NegativeHeight=4)
    ActiveRegionSize=0.1
    Debug=(bDrawDebug=true,bDraw2D=false,DebugScale=25,DebugSegments=20,DebugThickness=0.25)
    DriverAxis=(X=1,Y=0,Z=0)
    DriverItem=(Type=Bone,Name="thigh_r")
    FalloffRegionScaleFactors=(PositiveWidth=1,NegativeWidth=1.1,PositiveHeight=1,NegativeHeight=1)
    FalloffSize=0.45
    FlipHeightScaling=true
    FlipWidthScaling=true
    OptionalParentItem=(Type=Bone,Name="None")
    RotationOffset=(X=-90,Y=0,Z=0)
    ExecutePin=@n2.B
   )
   n5 = RigVMVariableNode() {variable=IsBadEmotePlaying}
   > &RigUnit_ModifyTransforms(
    ItemToModify="((Item=(Type=Bone,Name=\"dyn_skirt_front_c_MSTR\"),Transform=(Rotation=(X=0,Y=0,Z=0,W=1),Translation=(X=0,Y=-5,Z=15),Scale3D=(X=1,Y=1,Z=1))),(Item=(Type=Bone,Name=\"dyn_skirt_front_r_MSTR\"),Transform=(Rotation=(X=0,Y=0,Z=0,W=1),Translation=(X=0,Y=-8,Z=5),Scal ...(truncated)"
    ItemToModify.1.Transform.Rotation=&RigVMFunction_MathQuaternionFromEuler(Euler=(X=0,Y=0,Z=25)).Result
    Weight=RigVMDispatchNode(Values=(0,0), Index=&RigVMFunction_MathBoolToInteger(Value=@n5.Value).Result, Values.0=@n4.OutputParam) {template="DISPATCH_RigVMDispatch_SelectInt32(in Index,in Values,out Result)"}.Result
   )
   > &RigUnit_ModifyTransforms(
    ItemToModify=((Item=(Type=Bone,Name="dyn_skirt_r_MSTR"),Transform=(Rotation=(X=0,Y=0,Z=0,W=1),Translation=(X=-5,Y=0,Z=3),Scale3D=(X=1,Y=1,Z=1))))
    ItemToModify.0.Transform.Rotation=&RigVMFunction_MathQuaternionFromEuler(Euler=(X=0,Y=-35,Z=0)).Result
    Weight=@n4.OutputParam
   )
   > &RigUnit_ModifyTransforms(
    ItemToModify=((Item=(Type=Bone,Name="dyn_skirt_bk_r_MSTR"),Transform=(Rotation=(X=0,Y=0,Z=0,W=1),Translation=(X=0,Y=0,Z=0),Scale3D=(X=1,Y=1,Z=1))))
    ItemToModify.0.Transform.Rotation=&RigVMFunction_MathQuaternionFromEuler(Euler=(X=0,Y=-15,Z=0)).Result
    Weight=@n4.OutputParam
   )
   RigUnit_BeginExecution = &RigUnit_BeginExecution()
   RigVMFunction_Sequence = RigVMAggregateNode(ExecuteContext=@RigUnit_BeginExecution.ExecutePin)
   FN_DeformRig_FNC = RigVMFunctionReferenceNode(
    DeformRig_ Curve="(deformRig_l_upperarm_up_alpha_72_#1=False,deformRig_r_upperarm_up_alpha_73_#2=False,deformRig_l_upperarm_fr_alpha_74_#3=False,deformRig_r_upperarm_fr_alpha_75_CB51B8 ...(truncated)"
    DeformRig_Bool="(upperarm_1_#4=True,elbow_3_#5=True,wrist_7_#6=True,groin_8_#7=True,glute_17_#8=True,knee_9_40C408324CA261 ...(truncated)"
    DeformRig_Debug=(upperarm_debug_18_#4=False,wrist_debug_20_#6=False,lat_debug_26_#9=False,groinGlute_debug_27_#7=False)
    DeformRig_Value="(upperarm_up_18_#10=(Rotation=(X=0,Y=0,Z=0,W=1),Translation=(X=0,Y=0,Z=8),Scale3D=(X=0.7,Y=1,Z=1)),upperarm_fwd_fr_19_#11=(Rotation=(X=0,Y=0,Z=0,W=1),Translation=(X=0,Y=10,Z=0),Scale3D=(X=0.5,Y=1,Z= ...(truncated)"
    ExecuteContext=@RigVMFunction_Sequence.A
   )
   > n6 = &RigUnit_SphericalPoseReader(
    ActiveRegionScaleFactors=(PositiveWidth=3.5,NegativeWidth=0.5,PositiveHeight=3,NegativeHeight=3)
    ActiveRegionSize=0.1
    Debug=(bDrawDebug=false,DebugScale=25,DebugSegments=20,DebugThickness=0.25)
    DriverAxis=(X=-1,Y=0,Z=0)
    DriverItem=(Type=Bone,Name="thigh_l")
    FalloffRegionScaleFactors=(PositiveWidth=1,NegativeWidth=1,PositiveHeight=1,NegativeHeight=1)
    FalloffSize=0.45
    OptionalParentItem=(Type=Bone)
    RotationOffset=(X=90,Y=0,Z=0)
   )
   > RigVMVariableNode(Value=@n6.OutputParam) {variable=thigh l up}
   > n7 = &RigUnit_SphericalPoseReader(
    ActiveRegionScaleFactors=(PositiveWidth=3.5,NegativeWidth=0.5,PositiveHeight=3,NegativeHeight=3)
    ActiveRegionSize=0.1
    Debug=(bDrawDebug=false,DebugScale=25,DebugSegments=20,DebugThickness=0.25)
    DriverAxis=(X=1,Y=0,Z=0)
    DriverItem=(Type=Bone,Name="thigh_r")
    FalloffRegionScaleFactors=(PositiveWidth=1,NegativeWidth=1,PositiveHeight=1,NegativeHeight=1)
    FalloffSize=0.45
    FlipHeightScaling=false
    FlipWidthScaling=true
    OptionalParentItem=(Type=Bone)
    RotationOffset=(X=-90,Y=0,Z=0)
   )
   > RigVMVariableNode(Value=@n7.OutputParam) {variable=thigh r up}
   > n8 = &RigUnit_SphericalPoseReader(
    ActiveRegionScaleFactors=(PositiveWidth=2.5,NegativeWidth=0.5,PositiveHeight=2,NegativeHeight=2)
    ActiveRegionSize=0.1
    Debug=(bDrawDebug=false,DebugScale=25,DebugSegments=20,DebugThickness=0.25)
    DriverAxis=(X=-1,Y=0,Z=0)
    DriverItem=(Type=Bone,Name="thigh_l")
    FalloffRegionScaleFactors=(PositiveWidth=0,NegativeWidth=0.55,PositiveHeight=0,NegativeHeight=0)
    FalloffSize=0.2
    OptionalParentItem=(Type=Bone,Name="pelvis")
    RotationOffset=(X=-90,Y=0,Z=-65)
   )
   > RigVMVariableNode(Value=@n8.OutputParam) {variable=thigh l bk}
   > n9 = &RigUnit_SphericalPoseReader(
    ActiveRegionScaleFactors=(PositiveWidth=2.5,NegativeWidth=0.5,PositiveHeight=2,NegativeHeight=2)
    ActiveRegionSize=0.1
    Debug=(bDrawDebug=false,DebugScale=25,DebugSegments=20,DebugThickness=0.25)
    DriverAxis=(X=1,Y=0,Z=0)
    DriverItem=(Type=Bone,Name="thigh_r")
    FalloffRegionScaleFactors=(PositiveWidth=0,NegativeWidth=0.55,PositiveHeight=0,NegativeHeight=0)
    FalloffSize=0.2
    FlipHeightScaling=false
    FlipWidthScaling=true
    OptionalParentItem=(Type=Bone,Name="pelvis")
    RotationOffset=(X=90,Y=0,Z=-65)
   )
   > RigVMVariableNode(Value=@n9.OutputParam) {variable=thigh r bk}
   > n10 = &RigUnit_SphericalPoseReader(
    ActiveRegionScaleFactors=(PositiveWidth=3,NegativeWidth=2,PositiveHeight=3,NegativeHeight=3)
    ActiveRegionSize=0.12
    Debug=(bDrawDebug=false,DebugScale=25,DebugSegments=20,DebugThickness=0.25)
    DriverAxis=(X=-1,Y=0,Z=0)
    DriverItem=(Type=Bone,Name="thigh_l")
    FalloffRegionScaleFactors=(PositiveWidth=1,NegativeWidth=1,PositiveHeight=1,NegativeHeight=1)
    FalloffSize=0.45
    OptionalParentItem=(Type=Bone,Name="pelvis")
    RotationOffset=(X=180,Y=0,Z=0)
   )
   > RigVMVariableNode(Value=@n10.OutputParam) {variable=thigh l side}
   > n11 = &RigUnit_SphericalPoseReader(
    ActiveRegionScaleFactors=(PositiveWidth=3.5,NegativeWidth=3,PositiveHeight=3,NegativeHeight=4)
    ActiveRegionSize=0.12
    Debug=(bDrawDebug=false,DebugScale=25,DebugSegments=20,DebugThickness=0.25)
    DriverAxis=(X=1,Y=0,Z=0)
    DriverItem=(Type=Bone,Name="thigh_r")
    FalloffRegionScaleFactors=(PositiveWidth=1,NegativeWidth=1,PositiveHeight=1,NegativeHeight=1)
    FalloffSize=0.45
    FlipHeightScaling=false
    FlipWidthScaling=true
    OptionalParentItem=(Type=Bone,Name="pelvis")
    RotationOffset=(X=0,Y=0,Z=0)
   )
   > RigVMVariableNode(Value=@n11.OutputParam) {variable=thigh r side}
   > n12 = &RigUnit_SphericalPoseReader(
    ActiveRegionScaleFactors=(PositiveWidth=3.5,NegativeWidth=3,PositiveHeight=3.3,NegativeHeight=4)
    ActiveRegionSize=0.12
    Debug=(bDrawDebug=true,DebugScale=25,DebugSegments=20,DebugThickness=0.25)
    DriverAxis=(X=1,Y=0,Z=0)
    DriverItem=(Type=Bone,Name="thigh_r")
    FalloffRegionScaleFactors=(PositiveWidth=1,NegativeWidth=0.8,PositiveHeight=1,NegativeHeight=1)
    FalloffSize=0.5
    FlipHeightScaling=false
    FlipWidthScaling=true
    OptionalParentItem=(Type=Bone,Name="pelvis")
    RotationOffset=(X=170,Y=5,Z=0)
   )
   > RigVMVariableNode(Value=@n12.OutputParam) {variable=thigh r in}
   > n13 = &RigUnit_SphericalPoseReader(
    ActiveRegionScaleFactors=(PositiveWidth=3.5,NegativeWidth=3.2,PositiveHeight=3.4,NegativeHeight=3)
    ActiveRegionSize=0.12
    Debug=(bDrawDebug=false,DebugScale=25,DebugSegments=20,DebugThickness=0.25)
    DriverAxis=(X=-1,Y=0,Z=0)
    DriverItem=(Type=Bone,Name="thigh_l")
    FalloffRegionScaleFactors=(PositiveWidth=1,NegativeWidth=0.7,PositiveHeight=1,NegativeHeight=1)
    FalloffSize=0.5
    OptionalParentItem=(Type=Bone,Name="pelvis")
    RotationOffset=(X=-10,Y=5,Z=0)
   )
   > RigVMVariableNode(Value=@n13.OutputParam) {variable=thigh l in}
   > n14 = &RigUnit_SphericalPoseReader(
    ActiveRegionScaleFactors=(PositiveWidth=3.5,NegativeWidth=0.5,PositiveHeight=3,NegativeHeight=3)
    ActiveRegionSize=0.1
    Debug=(bDrawDebug=false,DebugScale=25,DebugSegments=20,DebugThickness=0.25)
    DriverAxis=(X=-1,Y=0,Z=0)
    DriverItem=(Type=Bone,Name="calf_l")
    FalloffRegionScaleFactors=(PositiveWidth=1,NegativeWidth=1,PositiveHeight=1,NegativeHeight=1)
    FalloffSize=0.45
    OptionalParentItem=(Type=Bone,Name="thigh_l")
    RotationOffset=(X=-90,Y=0,Z=20)
   )
   > RigVMVariableNode(Value=@n14.OutputParam) {variable=calf l in}
   > n15 = &RigUnit_SphericalPoseReader(
    ActiveRegionScaleFactors=(PositiveWidth=3.5,NegativeWidth=0.5,PositiveHeight=3,NegativeHeight=3)
    ActiveRegionSize=0.1
    Debug=(bDrawDebug=false,DebugScale=25,DebugSegments=20,DebugThickness=0.25)
    DriverAxis=(X=1,Y=0,Z=0)
    DriverItem=(Type=Bone,Name="calf_r")
    FalloffRegionScaleFactors=(PositiveWidth=1,NegativeWidth=1,PositiveHeight=1,NegativeHeight=1)
    FalloffSize=0.45
    FlipHeightScaling=false
    FlipWidthScaling=true
    OptionalParentItem=(Type=Bone,Name="thigh_r")
    RotationOffset=(X=90,Y=0,Z=25)
   )
   > RigVMVariableNode(Value=@n15.OutputParam) {variable=calf r in}
   &RigUnit_ParentConstraint(
    AdvancedSettings=(RotationOrderForFilter=XZY)
    Child=(Type=Bone,Name="dyn_skirtH_front_l_MSTR")
    Filter=(TranslationFilter=(bX=true,bY=true,bZ=true),RotationFilter=(bX=True,bY=True,bZ=True),ScaleFilter=(bX=false,bY=false,bZ=false))
    Parents=((Item=(Type=Bone,Name="thigh_l"),Weight=1))
    Weight=1
    ExecutePin=@RigVMFunction_Sequence.B
   )
   > &RigUnit_ParentConstraint(
    AdvancedSettings=(RotationOrderForFilter=XZY)
    Child=(Type=Bone,Name="dyn_skirt_front_c_MSTR")
    Filter=(TranslationFilter=(bX=false,bY=false,bZ=false),RotationFilter=(bX=True,bY=True,bZ=True),ScaleFilter=(bX=false,bY=false,bZ=false))
    Parents=((Item=(Type=Bone,Name="thigh_l"),Weight=1),(Item=(Type=Bone,Name="thigh_r"),Weight=0))
    Weight=1
   )
   > &RigUnit_ParentConstraint(
    AdvancedSettings=(RotationOrderForFilter=XZY)
    Child=(Type=Bone,Name="dyn_skirt_front_r_MSTR")
    Filter=(TranslationFilter=(bX=true,bY=true,bZ=true),RotationFilter=(bX=True,bY=True,bZ=True),ScaleFilter=(bX=false,bY=false,bZ=false))
    Parents=((Item=(Type=Bone,Name="thigh_r"),Weight=1))
    Weight=1
   )
   > &RigUnit_ParentConstraint(
    AdvancedSettings=(RotationOrderForFilter=XZY)
    Child=(Type=Bone,Name="dyn_skirt_l_MSTR")
    Filter=(TranslationFilter=(bX=true,bY=true,bZ=true),RotationFilter=(bX=True,bY=True,bZ=True),ScaleFilter=(bX=false,bY=false,bZ=false))
    Parents=((Item=(Type=Bone,Name="thigh_l"),Weight=0.7),(Item=(Type=Bone,Name="pelvis"),Weight=0.3))
    Weight=1
   )
   > &RigUnit_ParentConstraint(
    AdvancedSettings=(RotationOrderForFilter=XZY)
    Child=(Type=Bone,Name="dyn_skirt_r_MSTR")
    Filter=(TranslationFilter=(bX=false,bY=false,bZ=false),RotationFilter=(bX=True,bY=True,bZ=True),ScaleFilter=(bX=false,bY=false,bZ=false))
    Parents=((Item=(Type=Bone,Name="thigh_r"),Weight=0.7),(Item=(Type=Bone,Name="pelvis"),Weight=0.3))
    Weight=1
   )
   > &RigUnit_ParentConstraint(
    AdvancedSettings=(RotationOrderForFilter=XZY)
    Child=(Type=Bone,Name="dyn_skirtH_front_l_05")
    Parents=((Item=(Type=Bone,Name="calf_l"),Weight=1))
    Weight=0.8
   )
   > &RigUnit_ParentConstraint(
    AdvancedSettings=(RotationOrderForFilter=XZY)
    Child=(Type=Bone,Name="dyn_skirt_front_r_04")
    Parents=((Item=(Type=Bone,Name="calf_r"),Weight=1))
    Weight=0.7
   )
   n16 = &RigVMFunction_Sequence(ExecuteContext=@RigVMFunction_Sequence.F)
   n17 = RigVMVariableNode() {variable=thigh l side}
   &RigUnit_ModifyTransforms(
    ItemToModify=((Item=(Type=Bone,Name="dyn_skirt_front_c_01"),Transform=(Rotation=(X=0,Y=-0.0436,Z=0,W=0.999),Translation=(X=0,Y=0,Z=1),Scale3D=(X=1,Y=1,Z=1))))
    WeightMinimum=0
    ExecutePin=@n16.A
    Weight=@n17.Value
   )
   > &RigUnit_ModifyTransforms(
    ItemToModify=((Item=(Type=Bone,Name="dyn_skirtH_front_l_MSTR"),Transform=(Rotation=(X=0,Y=-0.061,Z=0,W=0.9981),Translation=(X=0,Y=0,Z=0),Scale3D=(X=1,Y=1,Z=1))))
    WeightMinimum=0
    Weight=@n17.Value
   )
   > &RigUnit_ModifyTransforms(
    ItemToModify=((Item=(Type=Bone,Name="dyn_skirt_l_MSTR"),Transform=(Rotation=(X=0,Y=-0.1305,Z=0,W=0.9914),Translation=(X=0,Y=0,Z=0),Scale3D=(X=1,Y=1,Z=1))))
    WeightMinimum=0
    Weight=@n17.Value
   )
   > &RigUnit_ModifyTransforms(
    ItemToModify=((Item=(Type=Bone,Name="dyn_skirt_bk_l_MSTR"),Transform=(Rotation=(X=0.0023,Y=-0.0261,Z=0.0871,W=0.9959),Translation=(X=0,Y=0,Z=0),Scale3D=(X=1,Y=1,Z=1))))
    WeightMinimum=0
    Weight=@n17.Value
   )
   n18 = RigVMVariableNode() {variable=thigh r side}
   > &RigUnit_ModifyTransforms(
    ItemToModify=((Item=(Type=Bone,Name="dyn_skirt_front_r_MSTR"),Transform=(Rotation=(X=0,Y=-0.0262,Z=0,W=0.9997),Translation=(X=0,Y=0,Z=1),Scale3D=(X=1,Y=1,Z=1))))
    WeightMinimum=0
    Weight=@n18.Value
   )
   > &RigUnit_ModifyTransforms(
    ItemToModify=((Item=(Type=Bone,Name="dyn_skirt_r_MSTR"),Transform=(Rotation=(X=0,Y=-0.061,Z=0,W=0.9981),Translation=(X=0,Y=0,Z=0),Scale3D=(X=1,Y=1,Z=1))))
    WeightMinimum=0
    Weight=@n18.Value
   )
   > &RigUnit_ModifyTransforms(
    ItemToModify=((Item=(Type=Bone,Name="dyn_skirt_bk_r_MSTR"),Transform=(Rotation=(X=0,Y=0,Z=-0.0872,W=0.9962),Translation=(X=0,Y=0,Z=0),Scale3D=(X=1,Y=1,Z=1))))
    WeightMinimum=0
    Weight=@n18.Value
   )
   n19 = RigVMVariableNode() {variable=thigh l in}
   &RigUnit_ModifyTransforms(
    ItemToModify=((Item=(Type=Bone,Name="dyn_skirt_front_c_01"),Transform=(Rotation=(X=0,Y=-0.061,Z=0,W=0.9981),Translation=(X=0,Y=0,Z=3),Scale3D=(X=1,Y=1,Z=1))))
    WeightMinimum=0
    ExecutePin=@n16.B
    Weight=@n19.Value
   )
   > &RigUnit_ModifyTransforms(
    ItemToModify=((Item=(Type=Bone,Name="dyn_skirtH_front_l_MSTR"),Transform=(Rotation=(X=0.003,Y=-0.0697,Z=0.0435,W=0.9966),Translation=(X=0,Y=0,Z=3),Scale3D=(X=1,Y=1,Z=1))))
    WeightMinimum=0
    Weight=@n19.Value
   )
   > &RigUnit_ModifyTransforms(
    ItemToModify=((Item=(Type=Bone,Name="dyn_skirt_l_MSTR"),Transform=(Rotation=(X=0,Y=-0.1305,Z=0,W=0.9914),Translation=(X=0,Y=0,Z=0),Scale3D=(X=1,Y=1,Z=1))))
    WeightMinimum=0
    Weight=@n19.Value
   )
   > &RigUnit_ModifyTransforms(
    ItemToModify=((Item=(Type=Bone,Name="dyn_skirt_bk_l_MSTR"),Transform=(Rotation=(X=0.0034,Y=-0.026,Z=0.1305,W=0.9911),Translation=(X=0,Y=0,Z=0),Scale3D=(X=1,Y=1,Z=1))))
    WeightMinimum=0
    Weight=@n19.Value
   )
   > &RigUnit_ModifyTransforms(
    ItemToModify=((Item=(Type=Bone,Name="dyn_skirt_r_MSTR"),Transform=(Rotation=(X=0,Y=-0.0872,Z=0,W=0.9962),Translation=(X=0,Y=0,Z=0),Scale3D=(X=1,Y=1,Z=1))))
    WeightMinimum=0
    Weight=@n19.Value
   )
   n20 = RigVMVariableNode() {variable=thigh r in}
   > &RigUnit_ModifyTransforms(
    ItemToModify=((Item=(Type=Bone,Name="dyn_skirt_front_r_MSTR"),Transform=(Rotation=(X=0,Y=-0.0872,Z=0,W=0.9962),Translation=(X=0,Y=0,Z=1),Scale3D=(X=1,Y=1,Z=1))))
    WeightMinimum=0
    Weight=@n20.Value
   )
   > &RigUnit_ModifyTransforms(
    ItemToModify=((Item=(Type=Bone,Name="dyn_skirt_r_MSTR"),Transform=(Rotation=(X=0,Y=-0.0872,Z=0,W=0.9962),Translation=(X=0,Y=0,Z=0),Scale3D=(X=1,Y=1,Z=1))))
    WeightMinimum=0
    Weight=@n20.Value
   )
   > &RigUnit_ModifyTransforms(
    ItemToModify=((Item=(Type=Bone,Name="dyn_skirt_bk_r_MSTR"),Transform=(Rotation=(X=0,Y=0,Z=-0.0872,W=0.9962),Translation=(X=0,Y=0,Z=0),Scale3D=(X=1,Y=1,Z=1))))
    WeightMinimum=0
    Weight=@n20.Value
   )
   > &RigUnit_ModifyTransforms(
    ItemToModify=((Item=(Type=Bone,Name="dyn_skirt_l_MSTR"),Transform=(Rotation=(X=0.0057,Y=-0.1304,Z=0.0432,W=0.9905),Translation=(X=0,Y=0,Z=0),Scale3D=(X=1,Y=1,Z=1))))
    WeightMinimum=0
    Weight=@n20.Value
   )
   > &RigUnit_ModifyTransforms(
    ItemToModify=((Item=(Type=Bone,Name="dyn_skirt_bk_l_MSTR"),Transform=(Rotation=(X=0.0034,Y=-0.026,Z=0.1305,W=0.9911),Translation=(X=0,Y=0,Z=0),Scale3D=(X=1,Y=1,Z=1))))
    WeightMinimum=0
    Weight=@n20.Value
   )
   n21 = &RigVMFunction_MathBoolToFloat(Value=&RigVMFunction_MathBoolNot(Value=RigVMVariableNode() {variable=IsRBANEnabled}.Value).Result)
   &RigUnit_ParentConstraint(
    AdvancedSettings=(InterpolationType=Average,RotationOrderForFilter=XZY)
    Child=(Type=Bone,Name="dyn_skirt_front_r_05")
    Filter=(TranslationFilter=(bX=false,bY=false,bZ=false),RotationFilter=(bX=True,bY=True,bZ=True),ScaleFilter=(bX=True,bY=True,bZ=True))
    Parents=((Item=(Type=Bone,Name="calf_r"),Weight=1))
    Weight=@n21.Result
    ExecutePin=@RigVMFunction_Sequence.G
   )
   > &RigUnit_ParentConstraint(
    AdvancedSettings=(InterpolationType=Average,RotationOrderForFilter=XZY)
    Child=(Type=Bone,Name="dyn_skirt_front_l_05")
    Filter=(TranslationFilter=(bX=false,bY=false,bZ=false),RotationFilter=(bX=True,bY=True,bZ=True),ScaleFilter=(bX=True,bY=True,bZ=True))
    Parents=((Item=(Type=Bone,Name="calf_l"),Weight=1))
    Weight=@n21.Result
   )
   > &RigUnit_ParentConstraint(
    AdvancedSettings=(InterpolationType=Shortest,RotationOrderForFilter=XZY)
    Child=(Type=Bone,Name="dyn_skirt_bk_r_04")
    Filter=(TranslationFilter=(bX=false,bY=false,bZ=false),RotationFilter=(bX=true,bY=true,bZ=true),ScaleFilter=(bX=True,bY=True,bZ=True))
    Parents=((Item=(Type=Bone,Name="calf_r"),Weight=1))
    Weight=@n21.Result
   )
   > &RigUnit_ParentConstraint(
    AdvancedSettings=(InterpolationType=Average,RotationOrderForFilter=XZY)
    Child=(Type=Bone,Name="dyn_skirt_bk_l_04")
    Filter=(TranslationFilter=(bX=false,bY=false,bZ=false),RotationFilter=(bX=true,bY=true,bZ=true),ScaleFilter=(bX=True,bY=True,bZ=True))
    Parents=((Item=(Type=Bone,Name="calf_l"),Weight=1))
    Weight=@n21.Result
   )
   > &RigUnit_ParentConstraint(
    AdvancedSettings=(InterpolationType=Average,RotationOrderForFilter=XZY)
    Child=(Type=Bone,Name="dyn_skirt_l_04")
    Filter=(TranslationFilter=(bX=false,bY=false,bZ=false),RotationFilter=(bX=True,bY=True,bZ=True),ScaleFilter=(bX=True,bY=True,bZ=True))
    Parents=((Item=(Type=Bone,Name="calf_l"),Weight=1))
    Weight=@n21.Result
   )
   > &RigUnit_ParentConstraint(
    AdvancedSettings=(InterpolationType=Average,RotationOrderForFilter=XZY)
    Child=(Type=Bone,Name="dyn_skirt_r_04")
    Filter=(TranslationFilter=(bX=false,bY=false,bZ=false),RotationFilter=(bX=True,bY=True,bZ=True),ScaleFilter=(bX=True,bY=True,bZ=True))
    Parents=((Item=(Type=Bone,Name="calf_r"),Weight=1))
    Weight=@n21.Result
   )
   n22 = RigVMVariableNode() {variable=thigh r up}
   &RigUnit_ModifyTransforms(
    ItemToModify=((Item=(Type=Bone,Name="dyn_skirt_front_r_MSTR"),Transform=(Rotation=(X=0,Y=-0.0872,Z=0,W=0.9962),Translation=(X=0,Y=4,Z=2),Scale3D=(X=1,Y=1,Z=1))))
    WeightMinimum=0
    Weight=@n22.Value
    ExecutePin=@RigVMFunction_Sequence.D
   )
   > &RigUnit_ModifyTransforms(
    ItemToModify=((Item=(Type=Bone,Name="dyn_skirt_r_MSTR"),Transform=(Rotation=(X=0.0283,Y=-0.1274,Z=0.2146,W=0.9679),Translation=(X=-2,Y=0,Z=2),Scale3D=(X=1,Y=1,Z=1))))
    WeightMinimum=0
    Weight=@n22.Value
   )
   Remap = &RigVMFunction_MathDoubleRemap(SourceMaximum=1, SourceMinimum=0, TargetMaximum=0, TargetMinimum=1, Value=RigVMVariableNode() {variable=jumping}.Value)
   n23 = &RigVMFunction_AlphaInterp(
    InterpSpeedDecreasing=10
    InterpSpeedIncreasing=10
    bClampResult=true
    Value=&RigVMFunction_MathDoubleMul(B=RigVMVariableNode() {variable=calf r in}.Value, A=RigVMVariableNode() {variable=thigh r up}.Value).Result
    Scale=@Remap.Result
   )
   > &RigUnit_ModifyTransforms(
    ItemToModify="((Item=(Type=Bone,Name=\"dyn_skirt_r_MSTR\"),Transform=(Rotation=(X=0.0076,Y=-0.043,Z=0.1735,W=0.9839),Translation=(X=0,Y=0,Z=0),Scale3D=(X=1,Y=1,Z=1))),(Item=(Type=Bone,Name=\"dyn_skirt_r_04\"),Transform=(Rotation=(X=0.0262,Y=-0.0831,Z=0.2996,W=0.9501),Transl ...(truncated)"
    WeightMinimum=0
    Weight=@n23.Result
   )
   > &RigUnit_ModifyTransforms(
    ItemToModify="((Item=(Type=Bone,Name=\"dyn_skirt_bk_r_MSTR\"),Transform=(Rotation=(X=0,Y=-0.2164,Z=0,W=0.9763),Translation=(X=0,Y=0,Z=0),Scale3D=(X=1,Y=1,Z=1))),(Item=(Type=Bone,Name=\"dyn_skirt_bk_r_03\"),Transform=(Rotation=(X=0,Y=-0.1305,Z=0,W=0.9914),Translation=(X=0,Y= ...(truncated)"
    WeightMinimum=0
    Weight=@n23.Result
   )
   n24 = RigVMVariableNode() {variable=thigh l up}
   &RigUnit_ModifyTransforms(
    ItemToModify=((Item=(Type=Bone,Name="dyn_skirtH_front_l_MSTR"),Transform=(Rotation=(X=-0.017,Y=-0.1294,Z=-0.1294,W=0.983),Translation=(X=0,Y=10,Z=2),Scale3D=(X=1,Y=1,Z=1))))
    WeightMinimum=0
    Weight=@n24.Value
    ExecutePin=@RigVMFunction_Sequence.C
   )
   > &RigUnit_ModifyTransforms(
    ItemToModify=((Item=(Type=Bone,Name="dyn_skirt_l_MSTR"),Transform=(Rotation=(X=-0.0283,Y=-0.1274,Z=-0.2146,W=0.9679),Translation=(X=-2,Y=0,Z=3),Scale3D=(X=1,Y=1,Z=1))))
    WeightMinimum=0
    Weight=@n24.Value
   )
   Multiply = &RigVMFunction_MathDoubleMul(B=RigVMVariableNode() {variable=calf l in}.Value, A=RigVMVariableNode() {variable=thigh l up}.Value)
   > &RigUnit_ModifyTransforms(
    ItemToModify="((Item=(Type=Bone,Name=\"dyn_skirt_l_MSTR\"),Transform=(Rotation=(X=-0.0076,Y=-0.043,Z=-0.1735,W=0.9839),Translation=(X=-2,Y=0,Z=0),Scale3D=(X=1,Y=1,Z=1))),(Item=(Type=Bone,Name=\"dyn_skirt_l_04\"),Transform=(Rotation=(X=0,Y=-0.1736,Z=0,W=0.9848),Translation=( ...(truncated)"
    WeightMinimum=0
    Weight=@Multiply.Result
   )
   > &RigUnit_ModifyTransforms(
    ItemToModify="((Item=(Type=Bone,Name=\"dyn_skirt_bk_l_MSTR\"),Transform=(Rotation=(X=-0.0057,Y=-0.1304,Z=-0.0432,W=0.9905),Translation=(X=0,Y=0,Z=0),Scale3D=(X=1,Y=1,Z=1))),(Item=(Type=Bone,Name=\"dyn_skirt_bk_l_04\"),Transform=(Rotation=(X=0,Y=-0.1305,Z=0,W=0.9914),Transla ...(truncated)"
    WeightMinimum=0
    Weight=@Multiply.Result
   )
   n25 = RigVMVariableNode() {variable=thigh r up}
   n26 = RigVMVariableNode() {variable=thigh l up}
   AlphaInterp = &RigVMFunction_AlphaInterp(
    InterpSpeedDecreasing=10
    InterpSpeedIncreasing=20
    Scale=1
    bClampResult=true
    bMapRange=true
    Value=&RigVMFunction_MathDoubleSub(A=@n25.Value, B=@n26.Value).Result
   )
   Subtract = &RigVMFunction_MathDoubleSub(B=@n25.Value, A=@n26.Value)
   Add = &RigVMFunction_MathDoubleAdd(B=0.3, A=@Subtract.Result)
   n27 = &RigVMFunction_AlphaInterp(InterpSpeedDecreasing=10, InterpSpeedIncreasing=20, Scale=1, bClampResult=true, bMapRange=true, Value=@Add.Result)
   &RigUnit_ParentConstraint(
    AdvancedSettings=(RotationOrderForFilter=XZY)
    Child=(Type=Bone,Name="dyn_skirt_front_c_01")
    Filter=(TranslationFilter=(bX=false,bY=false,bZ=false),RotationFilter=(bX=True,bY=True,bZ=True),ScaleFilter=(bX=True,bY=True,bZ=True))
    Parents=((Item=(Type=Bone,Name="dyn_skirtH_front_l_01"),Weight=1),(Item=(Type=Bone,Name="dyn_skirt_front_r_01"),Weight=1))
    Weight=1
    Parents.1.Weight=@AlphaInterp.Result
    Parents.0.Weight=@n27.Result
    ExecutePin=@RigVMFunction_Sequence.E
   )
   n28 = RigVMVariableNode() {variable=thigh l up}
   > &RigUnit_ModifyTransforms(
    ItemToModify=((Item=(Type=Bone,Name="dyn_skirt_front_c_01"),Transform=(Rotation=(X=0,Y=0.0175,Z=0,W=0.9998),Translation=(X=0,Y=0,Z=0),Scale3D=(X=1,Y=1,Z=1))))
    WeightMinimum=0
    Weight=@n28.Value
   )
   > &RigUnit_ParentConstraint(
    AdvancedSettings=(RotationOrderForFilter=XZY)
    Child=(Type=Bone,Name="dyn_skirt_front_l_04")
    Filter=(TranslationFilter=(bX=false,bY=false,bZ=false),RotationFilter=(bX=True,bY=True,bZ=True),ScaleFilter=(bX=True,bY=True,bZ=True))
    Parents=((Item=(Type=Bone,Name="dyn_skirt_front_r_05"),Weight=1))
    Weight=RigVMVariableNode() {variable=thigh r up}.Value
   )
   > &RigUnit_ParentConstraint(
    AdvancedSettings=(RotationOrderForFilter=XZY)
    Child=(Type=Bone,Name="dyn_skirt_front_l_04")
    Filter=(TranslationFilter=(bX=false,bY=false,bZ=false),RotationFilter=(bX=True,bY=True,bZ=True),ScaleFilter=(bX=True,bY=True,bZ=True))
    Parents=((Item=(Type=Bone,Name="dyn_skirtH_front_l_05"),Weight=1))
    Weight=@n28.Value
   )
   &RigUnit_ModifyTransforms(
    ExecutePin=()
    ItemToModify="((Item=(Type=Bone,Name=\"dyn_skirt_front_c_MSTR\"),Transform=(Rotation=(X=0,Y=0,Z=0,W=1),Translation=(X=0,Y=-5,Z=15),Scale3D=(X=1,Y=1,Z=1))),(Item=(Type=Bone,Name=\"dyn_skirt_front_r_MSTR\"),Transform=(Rotation=(X=0,Y=0,Z=0,W=1),Translation=(X=0,Y=-8,Z=5),Scal ...(truncated)"
    Weight=1
   )
   &RigUnit_ModifyTransforms(
    ExecutePin=()
    ItemToModify=((Item=(Type=Bone,Name="dyn_skirt_front_l_05"),Transform=(Rotation=(X=0,Y=0,Z=0,W=1),Translation=(X=0,Y=0,Z=0),Scale3D=(X=1,Y=1,Z=1))))
    Weight=1
    WeightMinimum=0
   )
   &RigUnit_ParentConstraint(
    AdvancedSettings=(InterpolationType=Average,RotationOrderForFilter=XZY)
    Child=(Type=Bone,Name="dyn_skirtH_front_l_02")
    ExecutePin=()
    Parents=((Item=(Type=Bone,Name="thigh_l"),Weight=1))
    Weight=1
   )
  ]}
]
hierarchy [
 root {loc=<0 0 0>, rot=<0 0 0 1>}
 pelvis {parent=root, loc=<0 -0.7071 93.1191>, rot=<0.0127 0.707 -0.0127 -0.707>}
 spine_01 {parent=pelvis, loc=<6.7494 -0.6388 0>, rot=<0 0 0.0503 -0.9987>}
 spine_02 {parent=spine_01, loc=<7.2493 -0.0308 0>, rot=<0 0 0.0041 -1>}
 spine_03 {parent=spine_02, loc=<6.0084 -0.0167 0>, rot=<0 0 0.0027 -1>}
 spine_04 {parent=spine_03, loc=<8.0999 -0.0224 0>, rot=<0 0 0.0026 -1>}
 spine_05 {parent=spine_04, loc=<10.0119 -0.0184 0>, rot=<0 0 -0.0016 -1>}
 clavicle_l {parent=spine_05, loc=<4.3575 -0.1466 -2.3223>, rot=<0.0476 -0.7289 -0.0102 -0.6829>}
 upperarm_l {parent=clavicle_l, loc=<15.5111 0 0>, rot=<0.0221 -0.3319 -0.0095 -0.943>}
 lowerarm_l {parent=upperarm_l, loc=<26.647 0 0>, rot=<0 0 -0.2447 -0.9696>}
 hand_l {parent=lowerarm_l, loc=<22.344 0 0>, rot=<-0.6794 -0.0546 0.1067 -0.7239>}
 weapon_l {parent=hand_l, loc=<-0.1974 -3.4111 0.3811>, rot=<0 0 0 1>}
 thumb_01_l {parent=hand_l, loc=<1.3118 -0.3191 -2.459>, rot=<0.6546 -0.4042 0.0664 -0.6354>}
 thumb_02_l {parent=thumb_01_l, loc=<5.9336 -0.0448 0.0401>, rot=<-0.0088 0.005 0.2979 -0.9545>}
 thumb_03_l {parent=thumb_02_l, loc=<2.2223 0.3682 0>, rot=<0 0 0.0067 -1>}
 index_metacarpal_l {parent=hand_l, loc=<1.6535 0.0183 -2.96>, rot=<0.0158 -0.0124 -0.0173 -0.9996>}
 index_01_l {parent=index_metacarpal_l, loc=<8.3008 0.0357 -0.0524>, rot=<-0.0044 0.0485 0.0924 -0.9945>}
 index_02_l {parent=index_01_l, loc=<3.3587 -0.3004 0>, rot=<0 0 0.3225 -0.9466>}
 index_03_l {parent=index_02_l, loc=<2.5578 -0.123 0>, rot=<0 0 0.2559 -0.9667>}
 middle_metacarpal_l {parent=hand_l, loc=<2.1856 -0.1725 -0.5984>, rot=<0.006 0.0168 -0.0329 -0.9993>}
 middle_01_l {parent=middle_metacarpal_l, loc=<7.3275 -0.2447 0.0129>, rot=<-0.0208 0.0042 0.0858 -0.9961>}
 middle_02_l {parent=middle_01_l, loc=<4.016 0.1415 0>, rot=<0 0 0.3519 -0.936>}
 middle_03_l {parent=middle_02_l, loc=<2.6103 0.0517 0>, rot=<0 0 0.1707 -0.9853>}
 ring_metacarpal_l {parent=hand_l, loc=<2.2074 -0.3625 1.3638>, rot=<-0.0284 0.0722 -0.0458 -0.9959>}
 ring_01_l {parent=ring_metacarpal_l, loc=<7.0276 -0.6134 -0.2041>, rot=<-0.0484 -0.0225 0.1894 -0.9805>}
 ring_02_l {parent=ring_01_l, loc=<3.6174 0.4647 0>, rot=<0 0 0.2806 -0.9598>}
 ring_03_l {parent=ring_02_l, loc=<2.4152 0.114 0>, rot=<0 0 0.1189 -0.9929>}
 pinky_metacarpal_l {parent=hand_l, loc=<1.7616 -0.9753 3.3814>, rot=<-0.0397 0.0814 -0.069 -0.9935>}
 pinky_01_l {parent=pinky_metacarpal_l, loc=<7.1709 -0.4831 -0.1094>, rot=<-0.0676 -0.0132 0.2236 -0.9723>}
 pinky_02_l {parent=pinky_01_l, loc=<2.8274 0.0893 0>, rot=<0 0 0.2854 -0.9584>}
 pinky_03_l {parent=pinky_02_l, loc=<2.2046 0.0307 0>, rot=<0 0 0.1445 -0.9895>}
 hand_attach_l {parent=hand_l, loc=<8.51 -3 0>, rot=<0.7071 0.7071 0 0>}
 deform_wrist_dn_base_l {parent=hand_l, loc=<0 -2.2342 0>, rot=<0 0 0 -1>}
 deform_wrist_dn_l {parent=deform_wrist_dn_base_l, loc=<0 0 0>, rot=<0 0 0 1>}
 deform_wrist_up_base_l {parent=hand_l, loc=<0 2.2682 0>, rot=<0 0 0 -1>}
 deform_wrist_up_l {parent=deform_wrist_up_base_l, loc=<0 0 0>, rot=<0 0 0 1>}
 lowerarm_twist_02_l {parent=lowerarm_l, loc=<9.6199 0 0>, rot=<0 0 0 1>}
 lowerarm_twist_01_l {parent=lowerarm_l, loc=<17.8549 0 0>, rot=<0 0 0 1>}
 deform_elbow_in_base_l {parent=lowerarm_l, loc=<0.2201 5.3074 0.8925>, rot=<0.0388 -0.0107 -0.0004 -0.9992>}
 deform_elbow_in_l {parent=deform_elbow_in_base_l, loc=<0 0 0>, rot=<0 0 0 1>}
 deform_elbow_out_base_l {parent=lowerarm_l, loc=<0.0067 -3.9879 0.7908>, rot=<0.0388 -0.0107 -0.0004 -0.9992>}
 deform_elbow_out_l {parent=deform_elbow_out_base_l, loc=<0 0 0>, rot=<0 0 0 1>}
 upperarm_twist_01_l {parent=upperarm_l, loc=<0.0041 0 0>, rot=<0 0 0 1>}
 deform_upperarm_bk_base_l {parent=upperarm_twist_01_l, loc=<1.1164 -7.2153 -1.9401>, rot=<-0.001 0.009 0 -1>}
 deform_upperarm_bk_l {parent=deform_upperarm_bk_base_l, loc=<0 0 0>, rot=<0 0 0 1>}
 deform_upperarm_fr_base_l {parent=upperarm_twist_01_l, loc=<-0.3194 7.0563 -2.7568>, rot=<-0.001 0.009 0 -1>}
 deform_upperarm_fr_l {parent=deform_upperarm_fr_base_l, loc=<0 0 0>, rot=<0 0 0 1>}
 deform_upperarm_up_base_l {parent=upperarm_twist_01_l, loc=<-3.2811 -1.1459 3.2702>, rot=<-0.001 0.009 0 -1>}
 deform_upperarm_up_l {parent=deform_upperarm_up_base_l, loc=<0 0 0>, rot=<0 0 0 1>}
 upperarm_twist_02_l {parent=upperarm_l, loc=<13.7941 0 0>, rot=<0 0 0 1>}
 neck_01 {parent=spine_05, loc=<11.3138 0.1529 0>, rot=<0 0 -0.2003 -0.9797>}
 neck_02 {parent=neck_01, loc=<4.5627 0 0>, rot=<0 0 0.0504 -0.9987>}
 head {parent=neck_02, loc=<5.3914 0 0>, rot=<0 0 0.115 -0.9934>}
 clavicle_r {parent=spine_05, loc=<4.3575 -0.1466 2.3223>, rot=<0.7289 0.0476 -0.6829 0.0102>}
 upperarm_r {parent=clavicle_r, loc=<-15.5111 0 0>, rot=<0.0221 -0.3319 -0.0095 -0.943>}
 lowerarm_r {parent=upperarm_r, loc=<-26.647 0 0>, rot=<0 0 -0.2447 -0.9696>}
 hand_r {parent=lowerarm_r, loc=<-22.344 0 0>, rot=<-0.6794 -0.0546 0.1067 -0.7239>}
 thumb_01_r {parent=hand_r, loc=<-1.3118 0.3191 2.459>, rot=<0.6546 -0.4042 0.0664 -0.6354>}
 thumb_02_r {parent=thumb_01_r, loc=<-5.9336 0.0448 -0.0401>, rot=<-0.0088 0.005 0.2979 -0.9545>}
 thumb_03_r {parent=thumb_02_r, loc=<-2.2223 -0.3682 0>, rot=<0 0 0.0067 -1>}
 index_metacarpal_r {parent=hand_r, loc=<-1.6535 -0.0183 2.96>, rot=<0.0158 -0.0124 -0.0173 -0.9996>}
 index_01_r {parent=index_metacarpal_r, loc=<-8.3008 -0.0357 0.0524>, rot=<-0.0044 0.0485 0.0924 -0.9945>}
 index_02_r {parent=index_01_r, loc=<-3.3587 0.3004 0>, rot=<0 0 0.3225 -0.9466>}
 index_03_r {parent=index_02_r, loc=<-2.5578 0.123 0>, rot=<0 0 0.2559 -0.9667>}
 middle_metacarpal_r {parent=hand_r, loc=<-2.1856 0.1725 0.5984>, rot=<0.006 0.0168 -0.0329 -0.9993>}
 middle_01_r {parent=middle_metacarpal_r, loc=<-7.3275 0.2447 -0.0129>, rot=<-0.0208 0.0042 0.0858 -0.9961>}
 middle_02_r {parent=middle_01_r, loc=<-4.016 -0.1415 0>, rot=<0 0 0.3519 -0.936>}
 middle_03_r {parent=middle_02_r, loc=<-2.6103 -0.0517 0>, rot=<0 0 0.1707 -0.9853>}
 ring_metacarpal_r {parent=hand_r, loc=<-2.2074 0.3625 -1.3638>, rot=<-0.0284 0.0722 -0.0458 -0.9959>}
 ring_01_r {parent=ring_metacarpal_r, loc=<-7.0276 0.6134 0.2041>, rot=<-0.0484 -0.0225 0.1894 -0.9805>}
 ring_02_r {parent=ring_01_r, loc=<-3.6174 -0.4647 0>, rot=<0 0 0.2806 -0.9598>}
 ring_03_r {parent=ring_02_r, loc=<-2.4152 -0.114 0>, rot=<0 0 0.1189 -0.9929>}
 pinky_metacarpal_r {parent=hand_r, loc=<-1.7616 0.9753 -3.3814>, rot=<-0.0397 0.0814 -0.069 -0.9935>}
 pinky_01_r {parent=pinky_metacarpal_r, loc=<-7.1709 0.4831 0.1094>, rot=<-0.0676 -0.0132 0.2236 -0.9723>}
 pinky_02_r {parent=pinky_01_r, loc=<-2.8274 -0.0893 0>, rot=<0 0 0.2854 -0.9584>}
 pinky_03_r {parent=pinky_02_r, loc=<-2.2046 -0.0307 0>, rot=<0 0 0.1445 -0.9895>}
 weapon_r {parent=hand_r, loc=<0.1974 3.4111 -0.3811>, rot=<0 0 0 1>}
 hand_attach_r {parent=hand_r, loc=<-8.51 2.9999 0>, rot=<0 0 -0.7071 -0.7071>}
 deform_wrist_dn_base_r {parent=hand_r, loc=<-0.0004 2.2345 -0.0001>, rot=<0 0 0 -1>}
 deform_wrist_dn_r {parent=deform_wrist_dn_base_r, loc=<0 0 0>, rot=<0 0 0 1>}
 deform_wrist_up_base_r {parent=hand_r, loc=<0.0003 -2.2684 0.0001>, rot=<0 0 0 -1>}
 deform_wrist_up_r {parent=deform_wrist_up_base_r, loc=<0 0 0>, rot=<0 0 0 1>}
 lowerarm_twist_02_r {parent=lowerarm_r, loc=<-9.6199 0 0>, rot=<0 0 0 -1>}
 lowerarm_twist_01_r {parent=lowerarm_r, loc=<-17.8549 0 0>, rot=<0 0 0 -1>}
 deform_elbow_in_base_r {parent=lowerarm_r, loc=<-0.22 -5.3074 -0.8926>, rot=<0.0388 -0.0107 -0.0004 -0.9992>}
 deform_elbow_in_r {parent=deform_elbow_in_base_r, loc=<-0.0001 0 0.0002>, rot=<0 0 0 1>}
 deform_elbow_out_base_r {parent=lowerarm_r, loc=<-0.0067 3.9879 -0.7908>, rot=<0.0388 -0.0107 -0.0004 -0.9992>}
 deform_elbow_out_r {parent=deform_elbow_out_base_r, loc=<0 0 0>, rot=<0 0 0 1>}
 upperarm_twist_01_r {parent=upperarm_r, loc=<-0.0041 0 0>, rot=<0 0 0 -1>}
 deform_upperarm_bk_base_r {parent=upperarm_twist_01_r, loc=<-1.1166 7.2153 1.9404>, rot=<-0.001 0.009 0 -1>}
 deform_upperarm_bk_r {parent=deform_upperarm_bk_base_r, loc=<0 0 0>, rot=<0 0 0 1>}
 deform_upperarm_fr_base_r {parent=upperarm_twist_01_r, loc=<0.3193 -7.0563 2.7569>, rot=<-0.001 0.009 0 -1>}
 deform_upperarm_fr_r {parent=deform_upperarm_fr_base_r, loc=<0 0 0>, rot=<0 0 0 1>}
 deform_upperarm_up_base_r {parent=upperarm_twist_01_r, loc=<3.2808 1.1459 -3.27>, rot=<-0.001 0.009 0 -1>}
 deform_upperarm_up_r {parent=deform_upperarm_up_base_r, loc=<0 0 0>, rot=<0 0 0 -1>}
 upperarm_twist_02_r {parent=upperarm_r, loc=<-13.7941 0 0>, rot=<0 0 0 -1>}
 deform_pec_aim_base_l {parent=spine_04, loc=<7.8667 12.6198 -2.4123>, rot=<0.0103 -0.4602 0.3978 -0.7937>}
 deform_pec_aim_l {parent=deform_pec_aim_base_l, loc=<0 0 0>, rot=<0 0 0 1>}
 deform_pec_base_l {parent=deform_pec_aim_l, loc=<10.766 0 0>, rot=<0 0 0 1>}
 deform_pec_l {parent=deform_pec_base_l, loc=<0.055 0 0>, rot=<0 0 0 1>}
 deform_pec_aim_base_r {parent=spine_04, loc=<7.8662 12.6197 2.4123>, rot=<-0.4602 -0.0103 0.7937 0.3978>}
 deform_pec_aim_r {parent=deform_pec_aim_base_r, loc=<0 0 0>, rot=<0 0 0 1>}
 deform_pec_base_r {parent=deform_pec_aim_r, loc=<-10.7662 -0.0006 -0.0007>, rot=<0 0 0 1>}
 deform_pec_r {parent=deform_pec_base_r, loc=<0 0 0>, rot=<0 0 0 1>}
 thigh_l {parent=pelvis, loc=<-2.5693 -0.0109 -7.7501>, rot=<0.0013 0.0462 0.0155 -0.9988>}
 calf_l {parent=thigh_l, loc=<-39.039 0 0>, rot=<0 0 -0.0378 -0.9993>}
 foot_l {parent=calf_l, loc=<-42.5546 0 0>, rot=<0.0427 -0.0505 0.0328 -0.9973>}
 ball_l {parent=foot_l, loc=<-6.1599 13.8264 0.6806>, rot=<0.0001 -0.0001 -0.7063 -0.7079>}
 calf_twist_02_l {parent=calf_l, loc=<-18.8064 -0.0967 0.0397>, rot=<0.0237 0.0095 -0.0035 -0.9997>}
 calf_twist_01_l {parent=calf_l, loc=<-35.813 -0.0967 0.0397>, rot=<0.0237 0.0095 -0.0035 -0.9997>}
 deform_knee_in_base_l {parent=calf_l, loc=<1.311 -5.2925 -0.9166>, rot=<0.0007 -0.0051 -0.0013 -1>}
 deform_knee_in_l {parent=deform_knee_in_base_l, loc=<0 0 0>, rot=<0 0 0 1>}
 deform_knee_out_base_l {parent=calf_l, loc=<-0.0102 4 -0.866>, rot=<0.0007 -0.0051 -0.0013 -1>}
 deform_knee_out_l {parent=deform_knee_out_base_l, loc=<0 0 0>, rot=<0 0 0 1>}
 thigh_twist_01_l {parent=thigh_l, loc=<-21.629 -0.0919 -0.0002>, rot=<0.0013 0 0 -1>}
 deform_glute_base_l {parent=thigh_l, loc=<-5.9985 -9 -0.1362>, rot=<0.0049 0.0099 -0.4617 -0.887>}
 deform_glute_l {parent=deform_glute_base_l, loc=<0 0 0>, rot=<0 0 0 1>}
 deform_groin_base_l {parent=thigh_l, loc=<-2.1519 8.1149 -1.7885>, rot=<-0.0002 0.011 0 -0.9999>}
 deform_groin_l {parent=deform_groin_base_l, loc=<0 0 0>, rot=<0 0 0 1>}
 thigh_r {parent=pelvis, loc=<-2.5693 -0.0109 7.7501>, rot=<0.0462 -0.0013 0.9988 0.0155>}
 calf_r {parent=thigh_r, loc=<39.0389 0 0>, rot=<0 0 -0.0378 -0.9993>}
 foot_r {parent=calf_r, loc=<42.5546 0 0>, rot=<0.0427 -0.0505 0.0328 -0.9973>}
 ball_r {parent=foot_r, loc=<6.1599 -13.8264 -0.6806>, rot=<0.0001 -0.0001 -0.7063 -0.7079>}
 calf_twist_02_r {parent=calf_r, loc=<18.8064 0.0967 -0.0397>, rot=<0.0237 0.0095 -0.0035 -0.9997>}
 calf_twist_01_r {parent=calf_r, loc=<35.813 0.0967 -0.0397>, rot=<0.0237 0.0095 -0.0035 -0.9997>}
 deform_knee_in_base_r {parent=calf_r, loc=<-1.311 5.2925 0.9166>, rot=<0.0007 -0.0051 -0.0013 -1>}
 deform_knee_in_r {parent=deform_knee_in_base_r, loc=<0 0 0>, rot=<0 0 0 1>}
 deform_knee_out_base_r {parent=calf_r, loc=<0.0103 -4 0.866>, rot=<0.0007 -0.0051 -0.0013 -1>}
 deform_knee_out_r {parent=deform_knee_out_base_r, loc=<0 0 0>, rot=<0 0 0 1>}
 thigh_twist_01_r {parent=thigh_r, loc=<21.629 0.0919 0.0002>, rot=<0.0013 0 0 -1>}
 deform_glute_base_r {parent=thigh_r, loc=<5.9983 9 0.1362>, rot=<0.0049 0.0099 -0.4617 -0.887>}
 deform_glute_r {parent=deform_glute_base_r, loc=<0 0 0>, rot=<0 0 0 1>}
 deform_groin_base_r {parent=thigh_r, loc=<2.1519 -8.1149 1.7885>, rot=<-0.0002 0.011 0 -0.9999>}
 deform_groin_r {parent=deform_groin_base_r, loc=<0 0 0>, rot=<0 0 0 1>}
 dyn_skirtH_front_l_MSTR {parent=pelvis, loc=<1.5902 6.8033 -7.9776>, rot=<-0.001 -0.8696 -0.4878 -0.0764>}
 dyn_skirtH_front_l_01 {parent=dyn_skirtH_front_l_MSTR, loc=<0 0 0>, rot=<0 0 0 1>}
 dyn_skirtH_front_l_02 {parent=dyn_skirtH_front_l_01, loc=<6.8108 0 0.0818>, rot=<0 -0.0045 0 -1>}
 dyn_skirtH_front_l_03 {parent=dyn_skirtH_front_l_02, loc=<12.3088 0 0.2706>, rot=<0 -0.016 0 -0.9999>}
 dyn_skirtH_front_l_04 {parent=dyn_skirtH_front_l_03, loc=<19.197 0 0>, rot=<0 0 0 1>}
 dyn_skirtH_front_l_05 {parent=dyn_skirtH_front_l_04, loc=<15.7184 0 0>, rot=<0 0 0 1>}
 dyn_skirtH_front_l_06 {parent=dyn_skirtH_front_l_05, loc=<15.4094 0 0>, rot=<0 0 0 1>}
 dyn_skirt_l_MSTR {parent=pelvis, loc=<2.8759 -3.6123 -13.2704>, rot=<0.0097 -0.9916 0.0975 -0.0842>}
 dyn_skirt_l_01 {parent=dyn_skirt_l_MSTR, loc=<0 0 0>, rot=<0 0 0 1>}
 dyn_skirt_l_02 {parent=dyn_skirt_l_01, loc=<6.6219 0 0.0784>, rot=<0 0 0 1>}
 dyn_skirt_l_03 {parent=dyn_skirt_l_02, loc=<12.7172 0 0.2267>, rot=<0 0.0023 0 -1>}
 dyn_skirt_l_04 {parent=dyn_skirt_l_03, loc=<19.4579 0 0>, rot=<0 -0.0107 0 -0.9999>}
 dyn_skirt_l_05 {parent=dyn_skirt_l_04, loc=<16.0412 0 0>, rot=<0 0 0 1>}
 dyn_skirt_l_06 {parent=dyn_skirt_l_05, loc=<16.0474 0 0>, rot=<0 0 0 1>}
 dyn_skirt_bk_l_MSTR {parent=pelvis, loc=<3.9367 -11.2986 -8.3321>, rot=<0.0062 -0.8075 0.586 -0.0669>}
 dyn_skirt_bk_l_01 {parent=dyn_skirt_bk_l_MSTR, loc=<0 0 0>, rot=<0 0 0 1>}
 dyn_skirt_bk_l_02 {parent=dyn_skirt_bk_l_01, loc=<6.9305 0 0.5359>, rot=<0 0 0 1>}
 dyn_skirt_bk_l_03 {parent=dyn_skirt_bk_l_02, loc=<13.7617 0 0.075>, rot=<0 -0.0116 0 -0.9999>}
 dyn_skirt_bk_l_04 {parent=dyn_skirt_bk_l_03, loc=<20.099 0 0.404>, rot=<0 0.0009 0 -1>}
 dyn_skirt_bk_l_05 {parent=dyn_skirt_bk_l_04, loc=<15.8376 0 0.1124>, rot=<0 -0.0243 0 -0.9997>}
 dyn_skirt_bk_l_06 {parent=dyn_skirt_bk_l_05, loc=<15.6824 0 0>, rot=<0 0 0 1>}
 dyn_skirt_front_c_MSTR {parent=pelvis, loc=<1.5902 9.0391 0>, rot=<0.0251 0.7067 0.7067 0.0251>}
 dyn_skirt_front_c_01 {parent=dyn_skirt_front_c_MSTR, loc=<0 0 0>, rot=<0 0 0 1>}
 dyn_skirt_front_l_02 {parent=dyn_skirt_front_c_01, loc=<6.8108 0 0.0818>, rot=<0 -0.0045 0 -1>}
 dyn_skirt_front_l_03 {parent=dyn_skirt_front_l_02, loc=<12.3088 0 0.2706>, rot=<0 0.013 0 -0.9999>}
 dyn_skirt_front_l_04 {parent=dyn_skirt_front_l_03, loc=<19.197 0 0>, rot=<0 0 0 1>}
 dyn_skirt_front_l_05 {parent=dyn_skirt_front_l_04, loc=<15.7184 0 0>, rot=<0 -0.0125 0 -0.9999>}
 dyn_skirt_front_l_06 {parent=dyn_skirt_front_l_05, loc=<15.4094 0 0>, rot=<0 0 0 1>}
 dyn_skirt_bk_r_MSTR {parent=pelvis, loc=<3.9368 -11.2985 8.3321>, rot=<0.0669 -0.586 0.8075 -0.0062>}
 dyn_skirt_bk_r_01 {parent=dyn_skirt_bk_r_MSTR, loc=<0 0 0>, rot=<0 0 0 1>}
 dyn_skirt_bk_r_02 {parent=dyn_skirt_bk_r_01, loc=<6.9305 0 0.5359>, rot=<0 0 0 1>}
 dyn_skirt_bk_r_03 {parent=dyn_skirt_bk_r_02, loc=<13.7616 0 0.075>, rot=<0 -0.0116 0 -0.9999>}
 dyn_skirt_bk_r_04 {parent=dyn_skirt_bk_r_03, loc=<20.099 0.0001 0.404>, rot=<0 0.0009 0 -1>}
 dyn_skirt_bk_r_05 {parent=dyn_skirt_bk_r_04, loc=<15.8376 0 0.1124>, rot=<0 -0.0243 0 -0.9997>}
 dyn_skirt_bk_r_06 {parent=dyn_skirt_bk_r_05, loc=<15.6824 0 0>, rot=<0 0 0 1>}
 dyn_skirt_r_MSTR {parent=pelvis, loc=<2.8758 -3.6123 13.2704>, rot=<0.0842 -0.0975 0.9916 -0.0097>}
 dyn_skirt_r_01 {parent=dyn_skirt_r_MSTR, loc=<0 0 0>, rot=<0 0 0 1>}
 dyn_skirt_r_02 {parent=dyn_skirt_r_01, loc=<6.6219 0 0.0784>, rot=<0 0 0 1>}
 dyn_skirt_r_03 {parent=dyn_skirt_r_02, loc=<12.7173 0 0.2266>, rot=<0 0.0023 0 -1>}
 dyn_skirt_r_04 {parent=dyn_skirt_r_03, loc=<19.4578 0 0>, rot=<0 -0.0107 0 -0.9999>}
 dyn_skirt_r_05 {parent=dyn_skirt_r_04, loc=<16.0412 0 0>, rot=<0 0 0 -1>}
 dyn_skirt_r_06 {parent=dyn_skirt_r_05, loc=<16.0474 0 0>, rot=<0 0 0 -1>}
 dyn_skirt_front_r_MSTR {parent=pelvis, loc=<1.5902 6.8033 7.9776>, rot=<0.0845 0.4878 0.8688 0.0056>}
 dyn_skirt_front_r_01 {parent=dyn_skirt_front_r_MSTR, loc=<0 0 0>, rot=<0 0 0 1>}
 dyn_skirt_front_r_02 {parent=dyn_skirt_front_r_01, loc=<6.8108 0 0.0818>, rot=<0 -0.0045 0 -1>}
 dyn_skirt_front_r_03 {parent=dyn_skirt_front_r_02, loc=<12.3088 0 0.2706>, rot=<0 -0.016 0 -0.9999>}
 dyn_skirt_front_r_04 {parent=dyn_skirt_front_r_03, loc=<19.1971 0 0>, rot=<0 0 0 1>}
 dyn_skirt_front_r_05 {parent=dyn_skirt_front_r_04, loc=<15.7183 0 0>, rot=<0 0 0 1>}
 dyn_skirt_front_r_06 {parent=dyn_skirt_front_r_05, loc=<15.4095 0 0>, rot=<0 0 0 1>}
 attach {parent=root, loc=<0 0 0>, rot=<0 0 0 1>}
 ik_foot_root {parent=root, loc=<0 0 0>, rot=<0 0 0 1>}
 ik_foot_l {parent=ik_foot_root, loc=<15.2608 -4.4176 9.4432>, rot=<0.0357 0.7045 0.0284 -0.7082>}
 ik_foot_r {parent=ik_foot_root, loc=<-15.2608 -4.4176 9.4432>, rot=<0.7082 0.0284 -0.7045 0.0357>}
 ik_hand_root {parent=root, loc=<0 0 0>, rot=<0 0 0 1>}
 ik_hand_gun {parent=ik_hand_root, loc=<-51.9727 4.0883 102.6935>, rot=<0.6084 0.1793 0.4613 -0.6205>}
 ik_hand_l {parent=ik_hand_gun, loc=<53.0285 82.1781 35.2051>, rot=<-0.5102 -0.7906 -0.3387 0>}
]
variables [
 IsBadEmotePlaying {type=bool, default=False}
 IsCrouching {type=bool, default=False}
 IsRBANEnabled {type=bool, default=False}
 IsonGrindRail {type=bool, default=False}
 calf l in {type=double, default=0}
 calf r in {type=double, default=0}
 jumping {type=double, default=0}
 thigh l bk {type=double, default=0}
 thigh l in {type=double, default=0}
 thigh l side {type=double, default=0}
 thigh l up {type=double, default=0}
 thigh r bk {type=double, default=0}
 thigh r in {type=double, default=0}
 thigh r side {type=double, default=0}
 thigh r up {type=double, default=0}
]
