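# Physics asset for the crouch variant of the male medium Maroon ensemble body, providing skirt bodies and lower-body collision shapes.
# Seven capsule chains hang from the pelvis, with free-rotation top joints and tight twist and asymmetric swing limits below.
# Extra free-rotation constraints cross-link some chains.
# Pelvis and asymmetric leg shapes add body volume, while a huge root capsule sunk below the origin approximates a floor.
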PhysicsAsset /Game/Characters/Player/Male/Medium/Bodies/M_MED_Ensemble_Maroon/Meshes/M_MED_Ensemble_Maroon_Crouch_Physics
bodies {
 calf_l [{
   shapeType=Capsule
   length=36.6802
   radius=11.5
   shapeName=calf_l_capsule
   center=<-19.8108 -2.2629 0.8171>
   rotation=<rot 88.1868 178.7541 179.9999>
  }]
 calf_r [{
   shapeType=Capsule
   length=39.1802
   radius=9
   shapeName=calf_l_capsule
   center=<19.8042 2.2631 -0.6074>
   rotation=<rot 88.1868 178.7541 -0.0001>
  }]
 deform_knee_out_base_l [{shapeType=Sphere, radius=7, shapeName=deform_knee_out_base_l_sphere, center=<0 -5.3491 -0.8699>}]
 dyn_skirtH_front_l_01 [{shapeType=Capsule, length=12, radius=5, shapeName=dyn_skirtH_front_l_01_capsule, center=<1 0 2>, rotation=<rot 0 0 90>}]
 dyn_skirtH_front_l_02 [{shapeType=Capsule, length=10, radius=8, shapeName=dyn_skirtH_front_l_02_capsule, center=<7 0 8>, rotation=<rot 0 0 90>}]
 dyn_skirtH_front_l_03 [{shapeType=Capsule, length=16, radius=10, shapeName=dyn_skirtH_front_l_03_capsule, center=<5 0 10>, rotation=<rot 0 0 90>}]
 dyn_skirtH_front_l_04 [{shapeType=Capsule, length=16, radius=10, shapeName=dyn_skirtH_front_l_04_capsule, center=<3 0 9>, rotation=<rot 0 0 90>}]
 dyn_skirtH_front_l_05 [{shapeType=Capsule, length=16, radius=12, shapeName=dyn_skirtH_front_l_05_capsule, center=<3 0 9>, rotation=<rot 0 0 90>}]
 dyn_skirtH_front_l_06 [{shapeType=Capsule, length=16, radius=12, shapeName=dyn_skirtH_front_l_06_capsule, center=<3 0 5>, rotation=<rot 0 0 90>}]
 dyn_skirtH_front_l_MSTR [{shapeType=Capsule, length=12, radius=5, shapeName=dyn_skirtH_front_l_MSTR_capsule, center=<1 0 2>, rotation=<rot 0 0 90>}]
 dyn_skirt_bk_l_01 [{shapeType=Capsule, length=12, radius=5, shapeName=dyn_skirt_bk_l_01_capsule, center=<1 0 2>, rotation=<rot 0 0 90>}]
 dyn_skirt_bk_l_02 [{shapeType=Capsule, length=10, radius=6, shapeName=dyn_skirt_bk_l_02_capsule, center=<8 0 3>, rotation=<rot 0 0 110>}]
 dyn_skirt_bk_l_03 [{shapeType=Capsule, length=8.5, radius=11, shapeName=dyn_skirt_bk_l_03_capsule, center=<8 0 6>, rotation=<rot 0 0 110>}]
 dyn_skirt_bk_l_04 [{shapeType=Capsule, length=30, radius=10, shapeName=dyn_skirt_bk_l_04_capsule, center=<3 6 4>, rotation=<rot 0 0 110>}]
 dyn_skirt_bk_l_05 [{shapeType=Capsule, length=30, radius=10, shapeName=dyn_skirt_bk_l_05_capsule, center=<3 6 1>, rotation=<rot 0 0 110>}]
 dyn_skirt_bk_l_06 [{shapeType=Capsule, length=31.5, radius=10.5, shapeName=dyn_skirt_bk_l_06_capsule, center=<3 6 1>, rotation=<rot 0 0 110>}]
 dyn_skirt_bk_l_MSTR [{shapeType=Capsule, length=12, radius=5, shapeName=dyn_skirt_bk_l_MSTR_capsule, center=<1 0 2>, rotation=<rot 0 0 90>}]
 dyn_skirt_bk_r_01 [{shapeType=Capsule, length=12, radius=5, shapeName=dyn_skirt_bk_r_01_capsule, center=<1 0 2>, rotation=<rot 0 0 90>}]
 dyn_skirt_bk_r_02 [{shapeType=Capsule, length=10, radius=6, shapeName=dyn_skirt_bk_r_02_capsule, center=<8 0 2>, rotation=<rot 0 0 70>}]
 dyn_skirt_bk_r_03 [{shapeType=Capsule, length=8, radius=10, shapeName=dyn_skirt_bk_r_03_capsule, center=<8 0 2>, rotation=<rot 0 0 70>}]
 dyn_skirt_bk_r_04 [{shapeType=Capsule, length=30, radius=10, shapeName=dyn_skirt_bk_r_04_capsule, center=<3 -6 1>, rotation=<rot 0 0 70>}]
 dyn_skirt_bk_r_05 [{shapeType=Capsule, length=30, radius=10, shapeName=dyn_skirt_bk_r_05_capsule, center=<3 -6 0>, rotation=<rot 0 0 70>}]
 dyn_skirt_bk_r_06 [{shapeType=Capsule, length=30, radius=12, shapeName=dyn_skirt_bk_r_06_capsule, center=<3 -6 0>, rotation=<rot 0 0 70>}]
 dyn_skirt_bk_r_MSTR [{shapeType=Capsule, length=12, radius=5, shapeName=dyn_skirt_bk_r_MSTR_capsule, center=<1 0 2>, rotation=<rot 0 0 90>}]
 dyn_skirt_front_c_01 [{shapeType=Capsule, length=12, radius=5, shapeName=dyn_skirt_front_c_01_capsule, center=<1 0 2>, rotation=<rot 0 0 90>}]
 dyn_skirt_front_c_MSTR [{shapeType=Capsule, length=12, radius=5, shapeName=dyn_skirt_front_c_MSTR_capsule, center=<1 0 2>, rotation=<rot 0 0 90>}]
 dyn_skirt_front_l_02 [{shapeType=Capsule, length=10, radius=6, shapeName=dyn_skirt_front_l_02_capsule, center=<7 -4 7>, rotation=<rot 0.0004 0 90>}]
 dyn_skirt_front_l_03 [{shapeType=Capsule, length=15.5, radius=6.5, shapeName=dyn_skirt_front_l_03_capsule, center=<8 -8 5>, rotation=<rot 0.0004 0 90>}]
 dyn_skirt_front_l_04 [{shapeType=Capsule, length=16, radius=10, shapeName=dyn_skirt_front_l_04_capsule, center=<3 -8 6.0053>, rotation=<rot 0.0004 0 90>}]
 dyn_skirt_front_l_05 [{shapeType=Capsule, length=16, radius=10, shapeName=dyn_skirt_front_l_05_capsule, center=<3 -8 5>, rotation=<rot 0.0005 0 90>}]
 dyn_skirt_front_l_06 [{shapeType=Capsule, length=16, radius=12, shapeName=dyn_skirt_front_l_06_capsule, center=<3 -8 5>, rotation=<rot 0.0005 0 90>}]
 dyn_skirt_front_r_01 [{shapeType=Capsule, length=12, radius=5, shapeName=dyn_skirt_front_r_01_capsule, center=<1 0 2>, rotation=<rot 0 0 90>}]
 dyn_skirt_front_r_02 [{shapeType=Capsule, length=10, radius=6, shapeName=dyn_skirt_front_r_02_capsule, center=<7 0 7.5>, rotation=<rot 0 0 90>}]
 dyn_skirt_front_r_03 [{shapeType=Capsule, length=10, radius=8, shapeName=dyn_skirt_front_r_03_capsule, center=<5 0 6>, rotation=<rot 0 0 100>}]
 dyn_skirt_front_r_04 [{shapeType=Capsule, length=10, radius=10, shapeName=dyn_skirt_front_r_04_capsule, center=<3 0 6>, rotation=<rot 0 0 100>}]
 dyn_skirt_front_r_05 [{shapeType=Capsule, length=10, radius=12, shapeName=dyn_skirt_front_r_05_capsule, center=<3 0 6>, rotation=<rot 0 0 100>}]
 dyn_skirt_front_r_06 [{shapeType=Capsule, length=10, radius=12, shapeName=dyn_skirt_front_r_06_capsule, center=<3 0 6>, rotation=<rot 0 0 100>}]
 dyn_skirt_front_r_MSTR [{shapeType=Capsule, length=12, radius=5, shapeName=dyn_skirt_front_r_MSTR_capsule, center=<1 0 2>, rotation=<rot 0 0 90>}]
 dyn_skirt_l_01 [{shapeType=Capsule, length=12, radius=5, shapeName=dyn_skirt_l_01_capsule, center=<1 0 2>, rotation=<rot 0 0 90>}]
 dyn_skirt_l_02 [{shapeType=Capsule, length=23.5, radius=6.25, shapeName=dyn_skirt_l_02_capsule, center=<7 0 7.5>, rotation=<rot 0 0 90>}]
 dyn_skirt_l_03 [{shapeType=Capsule, length=23.5, radius=8.25, shapeName=dyn_skirt_l_03_capsule, center=<5 0 8.5>, rotation=<rot 0 0 90>}]
 dyn_skirt_l_04 [{shapeType=Capsule, length=27.75, radius=12.25, shapeName=dyn_skirt_l_04_capsule, center=<3 0 6>, rotation=<rot 0 0 90>}]
 dyn_skirt_l_05 [{shapeType=Capsule, length=28, radius=16, shapeName=dyn_skirt_l_05_capsule, center=<3 0 9>, rotation=<rot 0 0 90>}]
 dyn_skirt_l_06 [{shapeType=Capsule, length=28, radius=16, shapeName=dyn_skirt_l_06_capsule, center=<3 0 9>, rotation=<rot 0 0 90>}]
 dyn_skirt_l_MSTR [{shapeType=Capsule, length=12, radius=5, shapeName=dyn_skirt_l_MSTR_capsule, center=<1 0 2>, rotation=<rot 0 0 90>}]
 dyn_skirt_r_01 [{shapeType=Capsule, length=12, radius=5, shapeName=dyn_skirt_r_01_capsule, center=<1 0 2>, rotation=<rot 0 0 90>}]
 dyn_skirt_r_02 [{shapeType=Capsule, length=10, radius=6, shapeName=dyn_skirt_r_02_capsule, center=<7 0 7.5>, rotation=<rot 0 0 90>}]
 dyn_skirt_r_03 [{shapeType=Capsule, length=10, radius=8, shapeName=dyn_skirt_r_03_capsule, center=<5 0 8.5>, rotation=<rot 0 0 90>}]
 dyn_skirt_r_04 [{shapeType=Capsule, length=28, radius=12, shapeName=dyn_skirt_r_04_capsule, center=<3 0 6>, rotation=<rot 0 0 90>}]
 dyn_skirt_r_05 [{shapeType=Capsule, length=28, radius=16, shapeName=dyn_skirt_r_05_capsule, center=<3 0 6>, rotation=<rot 0 0 90>}]
 dyn_skirt_r_06 [{shapeType=Capsule, length=28, radius=16, shapeName=dyn_skirt_r_06_capsule, center=<3 0 6>, rotation=<rot 0 0 90>}]
 dyn_skirt_r_MSTR [{shapeType=Capsule, length=12, radius=5, shapeName=dyn_skirt_r_MSTR_capsule, center=<1 0 2>, rotation=<rot 0 0 90>}]
 foot_l [{shapeType=Capsule, length=13.5, radius=8.5327, shapeName=foot_l_capsule, center=<-4 7 -0.1218>, rotation=<rot -0.0026 -0.0284 91.4352>}]
 foot_r [{
   shapeType=Capsule
   length=13
   radius=13.2827
   shapeName=foot_l_capsule
   center=<10.5086 -6.0714 0.0982>
   rotation=<rot -0.6088 24.9647 -88.7003>
  }]
 pelvis [{shapeType=Capsule, length=7.1058, radius=10.4289, shapeName=pelvis_capsule, center=<1.5783 -1.8601 0>}]
 root [
  {shapeType=Sphere, radius=15, shapeName=root_sphere, center=<0 0 -1.6855>}
  {shapeType=Capsule, length=0, radius=420, shapeName=root_capsule, center=<0 0 -415>}
 ]
 thigh_l [{
   shapeType=Capsule
   length=29.4421
   radius=8.7852
   shapeName=thigh_l_capsule
   center=<-16.8609 -1.1921 -0.0423>
   rotation=<rot 88.7695 176.3577 178.2462>
  }]
 thigh_r [{
   shapeType=Capsule
   length=35.4421
   radius=9.0352
   shapeName=thigh_l_capsule
   center=<14.8902 1.257 0>
   rotation=<rot 88.7695 176.3577 -1.7538>
  }]
 thigh_twist_01_l [{
   shapeType=Capsule
   length=10.9926
   radius=11
   shapeName=thigh_twist_01_l_capsule
   center=<3.3434 -2.266 -0.2359>
   rotation=<rot 0.0751 -90.7897 -91.1101>
  }]
}
constraints [
 {
  bone1Name=pelvis
  bone2Name=root
  swing1LimitDegrees=45
  swing1Motion=Free
  swing2LimitDegrees=45
  swing2Motion=Free
  twistLimitDegrees=45
  twistMotion=Free
 }
 {
  bone1Name=calf_l
  bone2Name=thigh_l
  swing1LimitDegrees=45
  swing1Motion=Free
  swing2LimitDegrees=45
  swing2Motion=Free
  twistLimitDegrees=45
  twistMotion=Free
 }
 {
  bone1Name=foot_l
  bone2Name=calf_l
  swing1LimitDegrees=45
  swing1Motion=Free
  swing2LimitDegrees=45
  swing2Motion=Free
  twistLimitDegrees=45
  twistMotion=Free
 }
 {
  bone1Name=thigh_r
  bone2Name=pelvis
  swing1LimitDegrees=45
  swing1Motion=Free
  swing2LimitDegrees=45
  swing2Motion=Free
  twistLimitDegrees=45
  twistMotion=Free
 }
 {
  bone1Name=thigh_l
  bone2Name=pelvis
  swing1LimitDegrees=45
  swing1Motion=Free
  swing2LimitDegrees=45
  swing2Motion=Free
  twistLimitDegrees=45
  twistMotion=Free
 }
 {
  bone1Name=foot_r
  bone2Name=calf_r
  swing1LimitDegrees=45
  swing1Motion=Free
  swing2LimitDegrees=45
  swing2Motion=Free
  twistLimitDegrees=45
  twistMotion=Free
 }
 {
  bone1Name=calf_r
  bone2Name=thigh_r
  swing1LimitDegrees=45
  swing1Motion=Free
  swing2LimitDegrees=45
  swing2Motion=Free
  twistLimitDegrees=45
  twistMotion=Free
 }
 {bone1Name=dyn_skirt_front_l_06, bone2Name=dyn_skirt_front_l_05, swing1LimitDegrees=10, swing2LimitDegrees=70, twistLimitDegrees=1}
 {
  bone1Name=dyn_skirtH_front_l_MSTR
  bone2Name=pelvis
  swing1LimitDegrees=45
  swing1Motion=Free
  swing2LimitDegrees=45
  swing2Motion=Free
  twistLimitDegrees=45
  twistMotion=Free
 }
 {
  bone1Name=dyn_skirtH_front_l_01
  bone2Name=dyn_skirtH_front_l_MSTR
  swing1LimitDegrees=45
  swing1Motion=Free
  swing2LimitDegrees=45
  swing2Motion=Free
  twistLimitDegrees=45
  twistMotion=Free
 }
 {bone1Name=dyn_skirtH_front_l_02, bone2Name=dyn_skirtH_front_l_01, swing1LimitDegrees=10, swing2LimitDegrees=70, twistLimitDegrees=1}
 {bone1Name=dyn_skirtH_front_l_03, bone2Name=dyn_skirtH_front_l_02, swing1LimitDegrees=10, swing2LimitDegrees=70, twistLimitDegrees=1}
 {bone1Name=dyn_skirtH_front_l_04, bone2Name=dyn_skirtH_front_l_03, swing1LimitDegrees=10, swing2LimitDegrees=70, twistLimitDegrees=1}
 {bone1Name=dyn_skirtH_front_l_05, bone2Name=dyn_skirtH_front_l_04, swing1LimitDegrees=10, swing2LimitDegrees=70, twistLimitDegrees=1}
 {bone1Name=dyn_skirtH_front_l_06, bone2Name=dyn_skirtH_front_l_05, swing1LimitDegrees=10, swing2LimitDegrees=70, twistLimitDegrees=1}
 {
  bone1Name=dyn_skirt_l_MSTR
  bone2Name=pelvis
  swing1LimitDegrees=45
  swing1Motion=Free
  swing2LimitDegrees=45
  swing2Motion=Free
  twistLimitDegrees=45
  twistMotion=Free
 }
 {
  bone1Name=dyn_skirt_l_01
  bone2Name=dyn_skirt_l_MSTR
  swing1LimitDegrees=45
  swing1Motion=Free
  swing2LimitDegrees=45
  swing2Motion=Free
  twistLimitDegrees=45
  twistMotion=Free
 }
 {bone1Name=dyn_skirt_l_02, bone2Name=dyn_skirt_l_01, swing1LimitDegrees=10, swing2LimitDegrees=70, twistLimitDegrees=1}
 {bone1Name=dyn_skirt_l_03, bone2Name=dyn_skirt_l_02, swing1LimitDegrees=10, swing2LimitDegrees=70, twistLimitDegrees=1}
 {bone1Name=dyn_skirt_l_04, bone2Name=dyn_skirt_l_03, swing1LimitDegrees=10, swing2LimitDegrees=70, twistLimitDegrees=1}
 {bone1Name=dyn_skirt_l_05, bone2Name=dyn_skirt_l_04, swing1LimitDegrees=10, swing2LimitDegrees=70, twistLimitDegrees=1}
 {bone1Name=dyn_skirt_l_06, bone2Name=dyn_skirt_l_05, swing1LimitDegrees=10, swing2LimitDegrees=70, twistLimitDegrees=1}
 {
  bone1Name=dyn_skirt_bk_l_MSTR
  bone2Name=pelvis
  swing1LimitDegrees=45
  swing1Motion=Free
  swing2LimitDegrees=45
  swing2Motion=Free
  twistLimitDegrees=45
  twistMotion=Free
 }
 {
  bone1Name=dyn_skirt_bk_l_01
  bone2Name=dyn_skirt_bk_l_MSTR
  swing1LimitDegrees=45
  swing1Motion=Free
  swing2LimitDegrees=45
  swing2Motion=Free
  twistLimitDegrees=45
  twistMotion=Free
 }
 {bone1Name=dyn_skirt_bk_l_02, bone2Name=dyn_skirt_bk_l_01, swing1LimitDegrees=10, swing2LimitDegrees=80, twistLimitDegrees=1}
 {bone1Name=dyn_skirt_bk_l_03, bone2Name=dyn_skirt_bk_l_02, swing1LimitDegrees=10, swing2LimitDegrees=70, twistLimitDegrees=1}
 {bone1Name=dyn_skirt_bk_l_04, bone2Name=dyn_skirt_bk_l_03, swing1LimitDegrees=10, swing2LimitDegrees=70, twistLimitDegrees=1}
 {bone1Name=dyn_skirt_bk_l_05, bone2Name=dyn_skirt_bk_l_04, swing1LimitDegrees=10, swing2LimitDegrees=70, twistLimitDegrees=1}
 {bone1Name=dyn_skirt_bk_l_06, bone2Name=dyn_skirt_bk_l_05, swing1LimitDegrees=10, swing2LimitDegrees=70, twistLimitDegrees=1}
 {
  bone1Name=dyn_skirt_bk_r_MSTR
  bone2Name=pelvis
  swing1LimitDegrees=45
  swing1Motion=Free
  swing2LimitDegrees=45
  swing2Motion=Free
  twistLimitDegrees=45
  twistMotion=Free
 }
 {
  bone1Name=dyn_skirt_bk_r_01
  bone2Name=dyn_skirt_bk_r_MSTR
  swing1LimitDegrees=45
  swing1Motion=Free
  swing2LimitDegrees=45
  swing2Motion=Free
  twistLimitDegrees=45
  twistMotion=Free
 }
 {bone1Name=dyn_skirt_bk_r_02, bone2Name=dyn_skirt_bk_r_01, swing1LimitDegrees=10, swing2LimitDegrees=70, twistLimitDegrees=1}
 {bone1Name=dyn_skirt_bk_r_03, bone2Name=dyn_skirt_bk_r_02, swing1LimitDegrees=10, swing2LimitDegrees=70, twistLimitDegrees=1}
 {bone1Name=dyn_skirt_bk_r_04, bone2Name=dyn_skirt_bk_r_03, swing1LimitDegrees=10, swing2LimitDegrees=70, twistLimitDegrees=1}
 {bone1Name=dyn_skirt_bk_r_05, bone2Name=dyn_skirt_bk_r_04, swing1LimitDegrees=10, swing2LimitDegrees=70, twistLimitDegrees=1}
 {bone1Name=dyn_skirt_bk_r_06, bone2Name=dyn_skirt_bk_r_05, swing1LimitDegrees=10, swing2LimitDegrees=70, twistLimitDegrees=1}
 {
  bone1Name=dyn_skirt_r_MSTR
  bone2Name=pelvis
  swing1LimitDegrees=45
  swing1Motion=Free
  swing2LimitDegrees=45
  swing2Motion=Free
  twistLimitDegrees=45
  twistMotion=Free
 }
 {
  bone1Name=dyn_skirt_r_01
  bone2Name=dyn_skirt_r_MSTR
  swing1LimitDegrees=45
  swing1Motion=Free
  swing2LimitDegrees=45
  swing2Motion=Free
  twistLimitDegrees=45
  twistMotion=Free
 }
 {bone1Name=dyn_skirt_r_02, bone2Name=dyn_skirt_r_01, swing1LimitDegrees=10, swing2LimitDegrees=70, twistLimitDegrees=1}
 {bone1Name=dyn_skirt_r_03, bone2Name=dyn_skirt_r_02, swing1LimitDegrees=10, swing2LimitDegrees=70, twistLimitDegrees=1}
 {bone1Name=dyn_skirt_r_04, bone2Name=dyn_skirt_r_03, swing1LimitDegrees=10, swing2LimitDegrees=70, twistLimitDegrees=1}
 {bone1Name=dyn_skirt_r_05, bone2Name=dyn_skirt_r_04, swing1LimitDegrees=10, swing2LimitDegrees=70, twistLimitDegrees=1}
 {bone1Name=dyn_skirt_r_06, bone2Name=dyn_skirt_r_05, swing1LimitDegrees=10, swing2LimitDegrees=70, twistLimitDegrees=1}
 {
  bone1Name=dyn_skirt_front_r_MSTR
  bone2Name=pelvis
  swing1LimitDegrees=45
  swing1Motion=Free
  swing2LimitDegrees=45
  swing2Motion=Free
  twistLimitDegrees=45
  twistMotion=Free
 }
 {
  bone1Name=dyn_skirt_front_r_01
  bone2Name=dyn_skirt_front_r_MSTR
  swing1LimitDegrees=45
  swing1Motion=Free
  swing2LimitDegrees=45
  swing2Motion=Free
  twistLimitDegrees=45
  twistMotion=Free
 }
 {bone1Name=dyn_skirt_front_r_02, bone2Name=dyn_skirt_front_r_01, swing1LimitDegrees=10, swing2LimitDegrees=70, twistLimitDegrees=1}
 {bone1Name=dyn_skirt_front_r_03, bone2Name=dyn_skirt_front_r_02, swing1LimitDegrees=10, swing2LimitDegrees=70, twistLimitDegrees=1}
 {bone1Name=dyn_skirt_front_r_04, bone2Name=dyn_skirt_front_r_03, swing1LimitDegrees=10, swing2LimitDegrees=70, twistLimitDegrees=1}
 {bone1Name=dyn_skirt_front_r_05, bone2Name=dyn_skirt_front_r_04, swing1LimitDegrees=10, swing2LimitDegrees=70, twistLimitDegrees=1}
 {bone1Name=dyn_skirt_front_r_06, bone2Name=dyn_skirt_front_r_05, swing1LimitDegrees=10, swing2LimitDegrees=70, twistLimitDegrees=1}
 {
  bone1Name=dyn_skirt_front_c_MSTR
  bone2Name=pelvis
  swing1LimitDegrees=45
  swing1Motion=Free
  swing2LimitDegrees=45
  swing2Motion=Free
  twistLimitDegrees=45
  twistMotion=Free
 }
 {
  bone1Name=dyn_skirt_front_c_01
  bone2Name=dyn_skirt_front_c_MSTR
  swing1LimitDegrees=45
  swing1Motion=Free
  swing2LimitDegrees=45
  swing2Motion=Free
  twistLimitDegrees=45
  twistMotion=Free
 }
 {bone1Name=dyn_skirt_front_l_02, bone2Name=dyn_skirt_front_c_01, swing1LimitDegrees=10, swing2LimitDegrees=80, twistLimitDegrees=1}
 {bone1Name=dyn_skirt_front_l_03, bone2Name=dyn_skirt_front_l_02, swing1LimitDegrees=10, swing2LimitDegrees=70, twistLimitDegrees=1}
 {bone1Name=dyn_skirt_front_l_04, bone2Name=dyn_skirt_front_l_03, swing1LimitDegrees=10, swing2LimitDegrees=70, twistLimitDegrees=1}
 {bone1Name=dyn_skirt_front_l_05, bone2Name=dyn_skirt_front_l_04, swing1LimitDegrees=10, swing2LimitDegrees=70, twistLimitDegrees=1}
 {
  bone1Name=dyn_skirt_l_05
  bone2Name=dyn_skirtH_front_l_05
  swing1LimitDegrees=45
  swing1Motion=Free
  swing2LimitDegrees=45
  swing2Motion=Free
  twistLimitDegrees=45
  twistMotion=Free
 }
 {
  bone1Name=dyn_skirt_l_03
  bone2Name=dyn_skirtH_front_l_03
  swing1LimitDegrees=45
  swing1Motion=Free
  swing2LimitDegrees=45
  swing2Motion=Free
  twistLimitDegrees=45
  twistMotion=Free
 }
 {
  bone1Name=dyn_skirt_r_03
  bone2Name=dyn_skirt_front_r_03
  swing1LimitDegrees=45
  swing1Motion=Free
  swing2LimitDegrees=45
  swing2Motion=Free
  twistLimitDegrees=45
  twistMotion=Free
 }
 {
  bone1Name=dyn_skirt_front_r_04
  bone2Name=dyn_skirt_r_04
  swing1LimitDegrees=45
  swing1Motion=Free
  swing2LimitDegrees=45
  swing2Motion=Free
  twistLimitDegrees=45
  twistMotion=Free
 }
 {
  bone1Name=dyn_skirtH_front_l_04
  bone2Name=dyn_skirt_l_04
  swing1LimitDegrees=45
  swing1Motion=Free
  swing2LimitDegrees=45
  swing2Motion=Free
  twistLimitDegrees=45
  twistMotion=Free
 }
 {
  bone1Name=dyn_skirt_bk_l_04
  bone2Name=dyn_skirt_l_04
  swing1LimitDegrees=45
  swing1Motion=Free
  swing2LimitDegrees=45
  swing2Motion=Free
  twistLimitDegrees=45
  twistMotion=Free
 }
 {
  bone1Name=dyn_skirt_bk_r_04
  bone2Name=dyn_skirt_r_04
  swing1LimitDegrees=45
  swing1Motion=Free
  swing2LimitDegrees=45
  swing2Motion=Free
  twistLimitDegrees=45
  twistMotion=Free
 }
 {
  bone1Name=deform_knee_out_base_l
  bone2Name=calf_l
  swing1LimitDegrees=45
  swing1Motion=Free
  swing2LimitDegrees=45
  swing2Motion=Free
  twistLimitDegrees=45
  twistMotion=Free
 }
 {
  bone1Name=dyn_skirtH_front_l_04
  bone2Name=dyn_skirt_front_l_04
  swing1LimitDegrees=45
  swing1Motion=Free
  swing2LimitDegrees=45
  swing2Motion=Free
  twistLimitDegrees=45
  twistMotion=Free
 }
 {
  bone1Name=dyn_skirtH_front_l_05
  bone2Name=dyn_skirt_front_l_05
  swing1LimitDegrees=45
  swing1Motion=Free
  swing2LimitDegrees=45
  swing2Motion=Free
  twistLimitDegrees=45
  twistMotion=Free
 }
 {
  bone1Name=dyn_skirtH_front_l_06
  bone2Name=dyn_skirtH_front_l_03
  swing1LimitDegrees=45
  swing1Motion=Free
  swing2LimitDegrees=45
  swing2Motion=Free
  twistLimitDegrees=45
  twistMotion=Free
 }
 {
  bone1Name=dyn_skirtH_front_l_05
  bone2Name=dyn_skirtH_front_l_03
  swing1LimitDegrees=45
  swing1Motion=Free
  swing2LimitDegrees=45
  swing2Motion=Free
  twistLimitDegrees=45
  twistMotion=Free
 }
 {
  bone1Name=dyn_skirtH_front_l_04
  bone2Name=dyn_skirtH_front_l_03
  swing1LimitDegrees=45
  swing1Motion=Free
  swing2LimitDegrees=45
  swing2Motion=Free
  twistLimitDegrees=45
  twistMotion=Free
 }
 {
  bone1Name=dyn_skirt_l_06
  bone2Name=dyn_skirt_bk_l_06
  swing1LimitDegrees=45
  swing1Motion=Free
  swing2LimitDegrees=45
  swing2Motion=Free
  twistLimitDegrees=45
  twistMotion=Free
 }
 {
  bone1Name=dyn_skirtH_front_l_03
  bone2Name=dyn_skirt_l_03
  swing1LimitDegrees=45
  swing1Motion=Free
  swing2LimitDegrees=45
  swing2Motion=Free
  twistLimitDegrees=45
  twistMotion=Free
 }
 {
  bone1Name=thigh_twist_01_l
  bone2Name=thigh_l
  swing1LimitDegrees=45
  swing1Motion=Free
  swing2LimitDegrees=45
  swing2Motion=Free
  twistLimitDegrees=45
  twistMotion=Free
 }
]
