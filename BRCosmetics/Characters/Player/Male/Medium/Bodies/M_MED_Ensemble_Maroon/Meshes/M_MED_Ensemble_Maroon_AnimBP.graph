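# Cosmetic AnimBP for the Maroon ensemble body that animates its coat skirt over the pose copied from the parent mesh.
# A coat Control Rig and slide offsets feed a cached pose, branched into skydive, front-end, emote or gameplay RigidBody physics, the last adding AnimDynamics.
# Movement and sprint bone offsets finish the pose; older chains sit unconnected.
# EventGraph derives gravity, movement, thigh and emote-montage values each update.
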AnimBlueprint /BRCosmetics/Characters/Player/Male/Medium/Bodies/M_MED_Ensemble_Maroon/Meshes/M_MED_Ensemble_Maroon_AnimBP
target_skeleton=&/M_MED_Ensemble_Maroon_Skeleton
graphs [
 AnimGraph {nodes [
   n1 = Control Rig(Source=Copy Pose From Mesh() {bUseAttachedParent=true}) {
    alphaBoolBlend {blendOption=Linear}
    controlRigAssetReference {blueprintRigClass=/BRCosmetics/Characters/Player/Male/Medium/Bodies/M_MED_Ensemble_Maroon_Belt/Meshes/M_LRG_DeformRig_Ensemble_Maroon_Belt_ControlRig1.M_LRG_DeformRig_Ensemble_Maroon_Belt_ControlRig1_C}
   }
   n2 = "Transform (Modify) Bone - Bone: dyn_skirt_bk_r_MSTR"(ComponentPose=Local To Component(LocalPose=@n1), Alpha=1, Translation="-19.296,-0.000,0.000", Rotation="0.677,-33.326,32.263", Scale="1,1,1") {
    rotationMode=BMM_Additive
    rotationSpace=BCS_BoneSpace
    scaleMode=BMM_Additive
    scaleSpace=BCS_BoneSpace
    translationMode=BMM_Additive
    translationSpace=BCS_BoneSpace
    alphaBoolBlend {blendOption=Linear}
    alphaScaleBiasClamp {bInterpResult=true}
    boneToModify {boneName=dyn_skirt_bk_r_MSTR}
    rotation=<rot 0.6769 -33.326 32.2631>
    translation {x=-19.296}
   }
   n3 = "Transform (Modify) Bone - Bone: dyn_skirt_bk_r_01"(ComponentPose=@n2, Alpha=1, Translation="-6.086,13.980,-17.874", Rotation="13.839,40.313,-38.401", Scale="1,1,1") {
    rotationMode=BMM_Additive
    rotationSpace=BCS_BoneSpace
    scaleMode=BMM_Additive
    scaleSpace=BCS_BoneSpace
    translationMode=BMM_Additive
    translationSpace=BCS_BoneSpace
    alphaBoolBlend {blendOption=Linear}
    alphaScaleBiasClamp {bInterpResult=true}
    boneToModify {boneName=dyn_skirt_bk_r_01}
    rotation=<rot 13.8389 40.3131 -38.401>
    translation=<-6.086 13.98 -17.874>
   }
   n4 = "Transform (Modify) Bone - Bone: dyn_skirt_bk_l_01"(ComponentPose=@n3, Alpha=1, Translation="0,0,0", Rotation="-7.383,72.397,-71.270", Scale="1,1,1") {
    rotationMode=BMM_Additive
    rotationSpace=BCS_BoneSpace
    scaleMode=BMM_Additive
    scaleSpace=BCS_BoneSpace
    translationMode=BMM_Additive
    translationSpace=BCS_BoneSpace
    alphaBoolBlend {blendOption=Linear}
    alphaScaleBiasClamp {bInterpResult=true}
    boneToModify {boneName=dyn_skirt_bk_l_01}
    rotation=<rot -7.383 72.3971 -71.2701>
   }
   n5 = "Transform (Modify) Bone - Bone: dyn_skirt_bk_l_MSTR"(ComponentPose=@n4, Alpha=1, Translation="-39.020,3.289,-5.788", Rotation="9.898,86.657,-171.468", Scale="1,1,1") {
    rotationMode=BMM_Additive
    rotationSpace=BCS_BoneSpace
    scaleMode=BMM_Additive
    scaleSpace=BCS_BoneSpace
    translationMode=BMM_Additive
    translationSpace=BCS_BoneSpace
    alphaBoolBlend {blendOption=Linear}
    alphaScaleBiasClamp {bInterpResult=true}
    boneToModify {boneName=dyn_skirt_bk_l_MSTR}
    rotation=<rot 9.898 86.657 -171.468>
    translation=<-39.02 3.289 -5.788>
   }
   n6 = RigidBody!(ComponentPose=@n5, Alpha, OverrideWorldGravity:Vector="0,0,-980", ExternalForce) -> (Pose) {
    bClampLinearTranslationLimitToRefPose=true
    bOverrideWorldGravity=true
    overridePhysicsAsset=/Game/Characters/Player/Male/Medium/Bodies/M_MED_Ensemble_Maroon/Meshes/M_MED_Ensemble_Maroon_Crouch_Physics
    simulationSpace=BaseBoneSpace
    alphaBoolBlend {blendOption=Linear}
    baseBoneRef {boneName=pelvis}
    componentAppliedLinearAccClamp=<3000 3000 3000>
    componentLinearAccScale=<0.4 0.2 0.2>
    componentLinearVelScale=<0.3 0.3 0.3>
    overrideWorldGravity {z=-980}
    simSpaceSettings {
     maxAngularAcceleration=1000
     maxAngularVelocity=3
     maxLinearAcceleration=700
     maxLinearVelocity=300
     velocityScaleZ=0
     externalLinearDragV=<0.2 0.2>
    }
   }
   Blend Poses by bool(BlendPose_0=Component To Local(ComponentPose=@n6), BlendTime_0=0.3, BlendTime_1=0.3, bActiveValue=Get bIsSliding()) {blendType=Linear}
   Component To Local()
   n7 = Get thigh_rot_otr_r() -> (thigh_rot_otr_r:double)
   n8 = "Transform (Modify) Bone - Bone: dyn_skirt_front_r_MSTR"(ComponentPose=Local To Component(), Alpha=@n7, Translation="0,0,0", Rotation="20,15,0.000", Scale="1,1,1") {
    rotationMode=BMM_Additive
    rotationSpace=BCS_BoneSpace
    scaleMode=BMM_Additive
    scaleSpace=BCS_BoneSpace
    translationMode=BMM_Additive
    translationSpace=BCS_BoneSpace
    alphaBoolBlend {blendOption=Linear}
    boneToModify {boneName=dyn_skirt_front_r_MSTR}
    rotation {pitch=20, yaw=15}
   }
   n9 = "Transform (Modify) Bone - Bone: dyn_skirt_r_MSTR"(ComponentPose=@n8, Alpha=@n7, Translation="0,0,0", Rotation="20,15,0.000", Scale="1,1,1") {
    rotationMode=BMM_Additive
    rotationSpace=BCS_BoneSpace
    scaleMode=BMM_Additive
    scaleSpace=BCS_BoneSpace
    translationMode=BMM_Additive
    translationSpace=BCS_BoneSpace
    alphaBoolBlend {blendOption=Linear}
    boneToModify {boneName=dyn_skirt_r_MSTR}
    rotation {pitch=20, yaw=15}
   }
   n10 = Get thigh_rot_otr_l() -> (thigh_rot_otr_l:double)
   n11 = "Transform (Modify) Bone - Bone: dyn_skirtH_front_l_MSTR"(ComponentPose=@n9, Alpha=@n10, Translation="0,0,0", Rotation="20,15,0.000", Scale="1,1,1") {
    rotationMode=BMM_Additive
    rotationSpace=BCS_BoneSpace
    scaleMode=BMM_Additive
    scaleSpace=BCS_BoneSpace
    translationMode=BMM_Additive
    translationSpace=BCS_BoneSpace
    alphaBoolBlend {blendOption=Linear}
    boneToModify {boneName=dyn_skirtH_front_l_MSTR}
    rotation {pitch=20, yaw=15}
   }
   n12 = "Transform (Modify) Bone - Bone: dyn_skirt_l_MSTR"(ComponentPose=@n11, Alpha=@n10, Translation="0,0,0", Rotation="20,15,0.000", Scale="1,1,1") {
    rotationMode=BMM_Additive
    rotationSpace=BCS_BoneSpace
    scaleMode=BMM_Additive
    scaleSpace=BCS_BoneSpace
    translationMode=BMM_Additive
    translationSpace=BCS_BoneSpace
    alphaBoolBlend {blendOption=Linear}
    boneToModify {boneName=dyn_skirt_l_MSTR}
    rotation {pitch=20, yaw=15}
   }
   n13 = Get BwdMovement() -> (BwdMovement:double)
   n14 = "Transform (Modify) Bone - Bone: dyn_skirt_bk_l_MSTR"(ComponentPose=@n12, Alpha=@n13, Translation="0,0,3", Rotation="25,0,0", Scale="1,1,1") {
    rotationMode=BMM_Additive
    rotationSpace=BCS_BoneSpace
    scaleMode=BMM_Additive
    scaleSpace=BCS_BoneSpace
    translationMode=BMM_Additive
    translationSpace=BCS_BoneSpace
    alphaBoolBlend {blendOption=Linear}
    alphaScaleBiasClamp {bInterpResult=true}
    boneToModify {boneName=dyn_skirt_bk_l_MSTR}
    rotation {pitch=25}
    translation {z=3}
   }
   n15 = "Transform (Modify) Bone - Bone: dyn_skirt_bk_r_MSTR"(ComponentPose=@n14, Alpha=@n13, Translation="0,0,3", Rotation="25,0,0", Scale="1,1,1") {
    rotationMode=BMM_Additive
    rotationSpace=BCS_BoneSpace
    scaleMode=BMM_Additive
    scaleSpace=BCS_BoneSpace
    translationMode=BMM_Additive
    translationSpace=BCS_BoneSpace
    alphaBoolBlend {blendOption=Linear}
    alphaScaleBiasClamp {bInterpResult=true}
    boneToModify {boneName=dyn_skirt_bk_r_MSTR}
    rotation {pitch=25}
    translation {z=3}
   }
   n16 = "Transform (Modify) Bone - Bone: dyn_skirt_l_MSTR"(ComponentPose=@n15, Alpha=Get LeftMovement() -> (LeftMovement:double), Translation="0,0,3", Rotation="25,0,0", Scale="1,1,1") {
    rotationMode=BMM_Additive
    rotationSpace=BCS_BoneSpace
    scaleMode=BMM_Additive
    scaleSpace=BCS_BoneSpace
    translationMode=BMM_Additive
    translationSpace=BCS_BoneSpace
    alphaBoolBlend {blendOption=Linear}
    alphaScaleBiasClamp {bInterpResult=true}
    boneToModify {boneName=dyn_skirt_l_MSTR}
    rotation {pitch=25}
    translation {z=3}
   }
   n17 = Get RightMovement() -> (RightMovement:double)
   n18 = "Transform (Modify) Bone - Bone: dyn_skirt_r_MSTR"(ComponentPose=@n16, Alpha=@n17, Translation="0,0,3", Rotation="40,0,0", Scale="1,1,1") {
    rotationMode=BMM_Additive
    rotationSpace=BCS_BoneSpace
    scaleMode=BMM_Additive
    scaleSpace=BCS_BoneSpace
    translationMode=BMM_Additive
    translationSpace=BCS_BoneSpace
    alphaBoolBlend {blendOption=Linear}
    alphaScaleBiasClamp {bInterpResult=true}
    boneToModify {boneName=dyn_skirt_r_MSTR}
    rotation {pitch=40}
    translation {z=3}
   }
   n19 = "Transform (Modify) Bone - Bone: dyn_skirt_front_r_MSTR"(ComponentPose=@n18, Alpha=@n17, Translation="0,0,0", Rotation="40,0,0", Scale="1,1,1") {
    rotationMode=BMM_Additive
    rotationSpace=BCS_BoneSpace
    scaleMode=BMM_Additive
    scaleSpace=BCS_BoneSpace
    translationMode=BMM_Additive
    translationSpace=BCS_BoneSpace
    alphaBoolBlend {blendOption=Linear}
    alphaScaleBiasClamp {bInterpResult=true}
    boneToModify {boneName=dyn_skirt_front_r_MSTR}
    rotation {pitch=40}
   }
   n20 = Get FwdMovement() -> (FwdMovement:double)
   n21 = "Transform (Modify) Bone - Bone: dyn_skirtH_front_l_MSTR"(ComponentPose=@n19, Alpha=@n20, Translation="0,0,0", Rotation="20,0,0", Scale="1,1,1") {
    rotationMode=BMM_Additive
    rotationSpace=BCS_BoneSpace
    scaleMode=BMM_Additive
    scaleSpace=BCS_BoneSpace
    translationMode=BMM_Additive
    translationSpace=BCS_BoneSpace
    alphaBoolBlend {blendOption=Linear}
    alphaScaleBiasClamp {bInterpResult=true}
    boneToModify {boneName=dyn_skirtH_front_l_MSTR}
    rotation {pitch=20}
   }
   n22 = "Transform (Modify) Bone - Bone: dyn_skirt_front_r_MSTR"(ComponentPose=@n21, Alpha=@n20, Translation="0,0,3", Rotation="35,0,0", Scale="1,1,1") {
    rotationMode=BMM_Additive
    rotationSpace=BCS_BoneSpace
    scaleMode=BMM_Additive
    scaleSpace=BCS_BoneSpace
    translationMode=BMM_Additive
    translationSpace=BCS_BoneSpace
    alphaBoolBlend {blendOption=Linear}
    alphaScaleBiasClamp {bInterpResult=true}
    boneToModify {boneName=dyn_skirt_front_r_MSTR}
    rotation {pitch=35}
    translation {z=3}
   }
   n23 = "Transform (Modify) Bone - Bone: dyn_skirt_bk_l_MSTR"(ComponentPose=@n22, Alpha=@n20, Translation="0,0,3", Rotation="15,0,0", Scale="1,1,1") {
    rotationMode=BMM_Additive
    rotationSpace=BCS_BoneSpace
    scaleMode=BMM_Additive
    scaleSpace=BCS_BoneSpace
    translationMode=BMM_Additive
    translationSpace=BCS_BoneSpace
    alphaBoolBlend {blendOption=Linear}
    alphaScaleBiasClamp {bInterpResult=true}
    boneToModify {boneName=dyn_skirt_bk_l_MSTR}
    rotation {pitch=15}
    translation {z=3}
   }
   n24 = "Transform (Modify) Bone - Bone: dyn_skirt_bk_r_MSTR"(ComponentPose=@n23, Alpha=@n20, Translation="0,0,3", Rotation="15,0,0", Scale="1,1,1") {
    rotationMode=BMM_Additive
    rotationSpace=BCS_BoneSpace
    scaleMode=BMM_Additive
    scaleSpace=BCS_BoneSpace
    translationMode=BMM_Additive
    translationSpace=BCS_BoneSpace
    alphaBoolBlend {blendOption=Linear}
    alphaScaleBiasClamp {bInterpResult=true}
    boneToModify {boneName=dyn_skirt_bk_r_MSTR}
    rotation {pitch=15}
    translation {z=3}
   }
   n25 = "Transform (Modify) Bone - Bone: dyn_skirt_front_c_MSTR"(ComponentPose=@n24, Alpha=@n20, Translation="0,0,0", Rotation="20,0,0", Scale="1,1,1") {
    rotationMode=BMM_Additive
    rotationSpace=BCS_BoneSpace
    scaleMode=BMM_Additive
    scaleSpace=BCS_BoneSpace
    translationMode=BMM_Additive
    translationSpace=BCS_BoneSpace
    alphaBoolBlend {blendOption=Linear}
    alphaScaleBiasClamp {bInterpResult=true}
    boneToModify {boneName=dyn_skirt_front_c_MSTR}
    rotation {pitch=20}
   }
   n26 = "Transform (Modify) Bone - Bone: dyn_skirt_l_01"(ComponentPose=@n25, Alpha=1, Translation="0,5,0", Rotation="0,0,0", Scale="1,1,1") {
    rotationMode=BMM_Additive
    translationMode=BMM_Additive
    alphaBoolBlend {blendOption=Linear}
    boneToModify {boneName=dyn_skirt_l_01}
    translation {y=5}
   }
   n27 = "Transform (Modify) Bone - Bone: dyn_skirt_r_01"(ComponentPose=@n26, Alpha=1, Translation="0,5,0", Rotation="0,0,0", Scale="1,1,1") {
    rotationMode=BMM_Additive
    translationMode=BMM_Additive
    alphaBoolBlend {blendOption=Linear}
    boneToModify {boneName=dyn_skirt_r_01}
    translation {y=5}
   }
   Component To Local(ComponentPose=@n27)
   Control Rig() {
    alphaBoolBlend {blendOption=Linear}
    controlRigAssetReference {blueprintRigClass=/Game/Characters/Player/Male/Medium/Bodies/M_MED_Ensemble_Maroon/Meshes/M_MED_Ensemble_Maroon_DeformRig_Master_ControlRig.M_MED_Ensemble_Maroon_DeformRig_Master_ControlRig_C}
   }
   n28 = Get Gravity Override()
   n29 = RigidBody(ComponentPose=Local To Component(LocalPose=Use cached pose 'SavedPose21276'()), OverrideWorldGravity=@n28) {
    bClampLinearTranslationLimitToRefPose=true
    bOverrideWorldGravity=true
    overridePhysicsAsset=&/M_MED_Ensemble_Maroon_Emote_Physics
    simulationSpace=BaseBoneSpace
    alphaBoolBlend {blendOption=Linear}
    baseBoneRef {boneName=pelvis}
    componentAppliedLinearAccClamp=<3000 3000 3000>
    componentLinearAccScale=<0.3 0.3 0.5>
    componentLinearVelScale=<0.3 0.3 0.5>
    overrideWorldGravity {z=-980}
    simSpaceSettings {
     maxAngularAcceleration=100
     maxAngularVelocity=3
     maxLinearAcceleration=700
     maxLinearVelocity=500
     velocityScaleZ=0
     worldAlpha=1
     externalLinearDragV=<0.2 0.2>
    }
   }
   n30 = Get isCustomEmotePlaying() -> (isCustomEmotePlaying:double)
   n31 = "Transform (Modify) Bone - Bone: dyn_skirtH_front_l_01"(ComponentPose=@n29, Alpha=@n30, Translation="0,5,3", Rotation="0,0,0", Scale="1,1,1") {translationMode=BMM_Additive, alphaBoolBlend {blendOption=Linear}, boneToModify {boneName=dyn_skirtH_front_l_01}, translation {y=5, z=3}}
   n32 = "Transform (Modify) Bone - Bone: dyn_skirt_front_c_01"(ComponentPose=@n31, Alpha=@n30, Translation="0,0,3", Rotation="0,0,0", Scale="1,1,1") {translationMode=BMM_Additive, alphaBoolBlend {blendOption=Linear}, boneToModify {boneName=dyn_skirt_front_c_01}, translation {z=3}}
   n33 = "Transform (Modify) Bone - Bone: dyn_skirt_front_r_01"(ComponentPose=@n32, Alpha=@n30, Translation="0.000,-0.000,3", Rotation="0,0,0", Scale="1,1,1") {translationMode=BMM_Additive, alphaBoolBlend {blendOption=Linear}, boneToModify {boneName=dyn_skirt_front_r_01}, translation {z=3}}
   n34 = RigidBody(ComponentPose=Local To Component(LocalPose=Use cached pose 'SavedPose21276'()), OverrideWorldGravity=@n28) {
    bClampLinearTranslationLimitToRefPose=true
    bOverrideWorldGravity=true
    overridePhysicsAsset=&/M_MED_Ensemble_Maroon_FE_Physics
    simulationSpace=BaseBoneSpace
    alphaBoolBlend {blendOption=Linear}
    baseBoneRef {boneName=pelvis}
    componentAppliedLinearAccClamp=<3000 3000 3000>
    componentLinearAccScale=<0.3 0.3 0.5>
    componentLinearVelScale=<0.3 0.3 0.5>
    overrideWorldGravity {z=-980}
    simSpaceSettings {
     maxAngularAcceleration=100
     maxAngularVelocity=3
     maxLinearAcceleration=700
     maxLinearVelocity=500
     velocityScaleZ=0
     worldAlpha=1
     externalLinearDragV=<0.2 0.2>
    }
   }
   n35 = RigidBody(ComponentPose=Local To Component(LocalPose=Use cached pose 'SavedPose21276'()), OverrideWorldGravity=@n28) {
    bClampLinearTranslationLimitToRefPose=true
    bOverrideWorldGravity=true
    overridePhysicsAsset=&/M_MED_Ensemble_Maroon_Crouch_Physics1
    simulationSpace=BaseBoneSpace
    alphaBoolBlend {blendOption=Linear}
    baseBoneRef {boneName=pelvis}
    componentAppliedLinearAccClamp=<3000 3000 3000>
    componentLinearAccScale=<0.1 0.1 0.1>
    componentLinearVelScale=<0.1 0.1 0.1>
    overrideWorldGravity {z=-980}
    simSpaceSettings {
     maxAngularAcceleration=100
     maxAngularVelocity=3
     maxLinearAcceleration=700
     maxLinearVelocity=500
     velocityScaleZ=0
     worldAlpha=1
     externalLinearDragV=<0.2 0.2>
    }
   }
   n36 = Get FwdMovement()
   n37 = "Transform (Modify) Bone - Bone: dyn_skirt_front_c_MSTR"(ComponentPose=@n35, Alpha=@n36, Translation="0,0,0.7", Rotation="15,0,0", Scale="1,1,1") {
    rotationMode=BMM_Additive
    rotationSpace=BCS_BoneSpace
    scaleMode=BMM_Additive
    scaleSpace=BCS_BoneSpace
    translationMode=BMM_Additive
    translationSpace=BCS_BoneSpace
    alphaBoolBlend {blendOption=Linear}
    alphaScaleBiasClamp {bInterpResult=true}
    boneToModify {boneName=dyn_skirt_front_c_MSTR}
    rotation {pitch=15}
    translation {z=0.7}
   }
   n38 = "Transform (Modify) Bone - Bone: dyn_skirtH_front_l_MSTR"(ComponentPose=@n37, Alpha=@n36, Translation="0,0,0.7", Rotation="15,0,0", Scale="1,1,1") {
    rotationMode=BMM_Additive
    rotationSpace=BCS_BoneSpace
    scaleMode=BMM_Additive
    scaleSpace=BCS_BoneSpace
    translationMode=BMM_Additive
    translationSpace=BCS_BoneSpace
    alphaBoolBlend {blendOption=Linear}
    alphaScaleBiasClamp {bInterpResult=true}
    boneToModify {boneName=dyn_skirtH_front_l_MSTR}
    rotation {pitch=15}
    translation {z=0.7}
   }
   n39 = Anim Dynamics - Bone: dyn_skirt_front_l_04(ComponentPose=@n38, Alpha=1, ExternalForce="0,0,0") {
    angularDampingOverride=0.99
    angularSpringConstant=50
    bAngularSpring=true
    bGravityOverrideInSimSpace=true
    bOverrideAngularDamping=true
    simulationSpace=RootRelative
    alphaBoolBlend {blendOption=Linear}
    boundBone {boneName=dyn_skirt_front_l_04}
    componentAppliedLinearAccClamp=<3500 3500 3500>
    componentLinearAccScale=<0.4 0.4 0.8>
    componentLinearVelScale=<0.4 0.4 0.8>
    gravityOverride {z=-810}
    physicsBodyDefinitions [{
      sphereCollisionRadius=0
      boundBone {boneName=dyn_skirt_front_l_04}
      boxExtents=<20 25 25>
      constraintSetup {angularLimitsMax=<5 0 5>, angularLimitsMin=<-5 -35 -35>, angularTarget=<1 0 0>}
      localJointOffset=<3 0 0>
     }]
    relativeSpaceBone {boneName=root}
   }
   n40 = Anim Dynamics - Bone: dyn_skirt_front_l_05(ComponentPose=@n39, Alpha=1, ExternalForce="0,0,0") {
    angularDampingOverride=0.99
    angularSpringConstant=50
    bAngularSpring=true
    bGravityOverrideInSimSpace=true
    bOverrideAngularDamping=true
    simulationSpace=RootRelative
    alphaBoolBlend {blendOption=Linear}
    boundBone {boneName=dyn_skirt_front_l_05}
    componentAppliedLinearAccClamp=<3500 3500 3500>
    componentLinearAccScale=<0.4 0.4 0.8>
    componentLinearVelScale=<0.4 0.4 0.8>
    gravityOverride {z=-810}
    physicsBodyDefinitions [{
      sphereCollisionRadius=0
      boundBone {boneName=dyn_skirt_front_l_05}
      boxExtents=<20 25 25>
      constraintSetup {angularLimitsMax=<5 5 5>, angularLimitsMin=<-5 -35 -35>, angularTarget=<1 0 0>}
      localJointOffset=<3 0 0>
     }]
    relativeSpaceBone {boneName=root}
   }
   n41 = Anim Dynamics - Bone: dyn_skirtH_front_l_05(ComponentPose=@n40, Alpha=10, ExternalForce="0,0,0") {
    angularDampingOverride=0.99
    angularSpringConstant=50
    bAngularSpring=true
    bGravityOverrideInSimSpace=true
    bOverrideAngularDamping=true
    simulationSpace=RootRelative
    alphaBoolBlend {blendOption=Linear}
    boundBone {boneName=dyn_skirtH_front_l_05}
    componentAppliedLinearAccClamp=<3500 3500 3500>
    componentLinearAccScale=<0.4 0.4 0.8>
    componentLinearVelScale=<0.4 0.4 0.8>
    gravityOverride {z=-810}
    physicsBodyDefinitions [{
      sphereCollisionRadius=0
      boundBone {boneName=dyn_skirtH_front_l_05}
      boxExtents=<20 25 25>
      constraintSetup {angularLimitsMax=<0 5 5>, angularLimitsMin=<0 -35 -25>, angularTarget=<1 0 0>}
      localJointOffset=<3 0 0>
     }]
    relativeSpaceBone {boneName=root}
   }
   n42 = Anim Dynamics - Bone: dyn_skirt_front_l_06(ComponentPose=@n41, Alpha=1, ExternalForce="0,0,0") {
    angularDampingOverride=0.99
    angularSpringConstant=50
    bAngularSpring=true
    bGravityOverrideInSimSpace=true
    bOverrideAngularDamping=true
    simulationSpace=RootRelative
    alphaBoolBlend {blendOption=Linear}
    boundBone {boneName=dyn_skirt_front_l_06}
    componentAppliedLinearAccClamp=<3500 3500 3500>
    componentLinearAccScale=<0.4 0.4 0.8>
    componentLinearVelScale=<0.4 0.4 0.8>
    gravityOverride {z=-810}
    physicsBodyDefinitions [{
      sphereCollisionRadius=0
      boundBone {boneName=dyn_skirt_front_l_06}
      boxExtents=<20 25 25>
      constraintSetup {angularLimitsMax=<5 5 5>, angularLimitsMin=<-5 -35 -35>, angularTarget=<1 0 0>}
      localJointOffset=<3 0 0>
     }]
    relativeSpaceBone {boneName=root}
   }
   n43 = Anim Dynamics - Bone: dyn_skirtH_front_l_06(ComponentPose=@n42, Alpha=10, ExternalForce="0,0,0") {
    angularDampingOverride=0.99
    angularSpringConstant=50
    bAngularSpring=true
    bGravityOverrideInSimSpace=true
    bOverrideAngularDamping=true
    simulationSpace=RootRelative
    alphaBoolBlend {blendOption=Linear}
    boundBone {boneName=dyn_skirtH_front_l_06}
    componentAppliedLinearAccClamp=<3500 3500 3500>
    componentLinearAccScale=<0.4 0.4 0.8>
    componentLinearVelScale=<0.4 0.4 0.8>
    gravityOverride {z=-810}
    physicsBodyDefinitions [{
      sphereCollisionRadius=0
      boundBone {boneName=dyn_skirtH_front_l_06}
      boxExtents=<20 25 25>
      constraintSetup {angularLimitsMax=<0 5 5>, angularLimitsMin=<0 -35 -25>, angularTarget=<1 0 0>}
      localJointOffset=<3 0 0>
     }]
    relativeSpaceBone {boneName=root}
   }
   n44 = Anim Dynamics - Bone: dyn_skirt_front_r_05(ComponentPose=@n43, Alpha=10, ExternalForce="0,0,0") {
    angularDampingOverride=0.99
    angularSpringConstant=50
    bAngularSpring=true
    bGravityOverrideInSimSpace=true
    bOverrideAngularDamping=true
    simulationSpace=RootRelative
    alphaBoolBlend {blendOption=Linear}
    boundBone {boneName=dyn_skirt_front_r_05}
    componentAppliedLinearAccClamp=<3500 3500 3500>
    componentLinearAccScale=<0.4 0.4 0.8>
    componentLinearVelScale=<0.4 0.4 0.8>
    gravityOverride {z=-810}
    physicsBodyDefinitions [{
      sphereCollisionRadius=0
      boundBone {boneName=dyn_skirt_front_r_05}
      boxExtents=<20 25 25>
      constraintSetup {angularLimitsMax=<0 5 25>, angularLimitsMin=<0 -35 -5>, angularTarget=<1 0 0>}
      localJointOffset=<3 0 0>
     }]
    relativeSpaceBone {boneName=root}
   }
   n45 = Anim Dynamics - Bone: dyn_skirt_front_r_06(ComponentPose=@n44, Alpha=10, ExternalForce="0,0,0") {
    angularDampingOverride=0.99
    angularSpringConstant=50
    bAngularSpring=true
    bGravityOverrideInSimSpace=true
    bOverrideAngularDamping=true
    simulationSpace=RootRelative
    alphaBoolBlend {blendOption=Linear}
    boundBone {boneName=dyn_skirt_front_r_06}
    componentAppliedLinearAccClamp=<3500 3500 3500>
    componentLinearAccScale=<0.4 0.4 0.8>
    componentLinearVelScale=<0.4 0.4 0.8>
    gravityOverride {z=-810}
    physicsBodyDefinitions [{
      sphereCollisionRadius=0
      boundBone {boneName=dyn_skirt_front_r_06}
      boxExtents=<20 25 25>
      constraintSetup {angularLimitsMax=<5 5 25>, angularLimitsMin=<-5 -35 -5>, angularTarget=<1 0 0>}
      localJointOffset=<3 0 0>
     }]
    relativeSpaceBone {boneName=root}
   }
   n46 = Blend Poses by bool(
    BlendPose_0=Blend Poses by bool(BlendPose_0=Component To Local(ComponentPose=@n33), BlendPose_1=Component To Local(ComponentPose=@n34), bActiveValue=Get bIsPlayingEmote()) {blendType=Linear}
    BlendPose_1=Component To Local(ComponentPose=@n45)
    bActiveValue=Get bIsInFrontEnd()
   ) {blendType=Linear}
   n47 = RigidBody(ComponentPose=Local To Component(LocalPose=Use cached pose 'SavedPose21276'()), OverrideWorldGravity=@n28) {
    bClampLinearTranslationLimitToRefPose=true
    bOverrideWorldGravity=true
    overridePhysicsAsset=&/M_MED_Ensemble_Maroon_Skydive_Physics
    simulationSpace=BaseBoneSpace
    alphaBoolBlend {blendOption=Linear}
    baseBoneRef {boneName=spine_02}
    componentAppliedLinearAccClamp=<3000 3000 3000>
    componentLinearAccScale=<0.2 0.2 0.4>
    componentLinearVelScale=<0.1 0.1 0.1>
    overrideWorldGravity {z=-980}
    simSpaceSettings {
     maxAngularAcceleration=1000
     maxAngularVelocity=3
     maxLinearAcceleration=700
     maxLinearVelocity=500
     velocityScaleZ=0
     externalLinearDragV=<0.2 0.2>
    }
   }
   n48 = Component To Local(ComponentPose=@n47)
   n49 = Get bIsSkyDiving()
   n50 = Get LeftMovement()
   n51 = "Transform (Modify) Bone - Bone: dyn_skirtH_front_l_MSTR"(
    ComponentPose=Local To Component(LocalPose=Blend Poses by bool(BlendPose_0=@n48, BlendPose_1=@n46, bActiveValue=@n49) {blendType=Linear})
    Alpha=@n50
    Translation="0,0,0"
    Rotation="5,0,0"
    Scale="1,1,1"
   ) {
    rotationMode=BMM_Additive
    rotationSpace=BCS_BoneSpace
    scaleMode=BMM_Additive
    scaleSpace=BCS_BoneSpace
    translationMode=BMM_Additive
    translationSpace=BCS_BoneSpace
    alphaBoolBlend {blendOption=Linear}
    alphaScaleBiasClamp {bInterpResult=true}
    boneToModify {boneName=dyn_skirtH_front_l_MSTR}
    rotation {pitch=5}
   }
   n52 = "Transform (Modify) Bone - Bone: dyn_skirt_l_MSTR"(ComponentPose=@n51, Alpha=@n50, Translation="0,0,0", Rotation="5,0,0", Scale="1,1,1") {
    rotationMode=BMM_Additive
    rotationSpace=BCS_BoneSpace
    scaleMode=BMM_Additive
    scaleSpace=BCS_BoneSpace
    translationMode=BMM_Additive
    translationSpace=BCS_BoneSpace
    alphaBoolBlend {blendOption=Linear}
    alphaScaleBiasClamp {bInterpResult=true}
    boneToModify {boneName=dyn_skirt_l_MSTR}
    rotation {pitch=5}
   }
   n53 = Get RightMovement()
   n54 = "Transform (Modify) Bone - Bone: dyn_skirt_r_MSTR"(ComponentPose=@n52, Alpha=@n53, Translation="0,0,0", Rotation="5,0,0", Scale="1,1,1") {
    rotationMode=BMM_Additive
    rotationSpace=BCS_BoneSpace
    scaleMode=BMM_Additive
    scaleSpace=BCS_BoneSpace
    translationMode=BMM_Additive
    translationSpace=BCS_BoneSpace
    alphaBoolBlend {blendOption=Linear}
    alphaScaleBiasClamp {bInterpResult=true}
    boneToModify {boneName=dyn_skirt_r_MSTR}
    rotation {pitch=5}
   }
   n55 = "Transform (Modify) Bone - Bone: dyn_skirt_front_r_MSTR"(ComponentPose=@n54, Alpha=@n53, Translation="0,0,0", Rotation="5,0,0", Scale="1,1,1") {
    rotationMode=BMM_Additive
    rotationSpace=BCS_BoneSpace
    scaleMode=BMM_Additive
    scaleSpace=BCS_BoneSpace
    translationMode=BMM_Additive
    translationSpace=BCS_BoneSpace
    alphaBoolBlend {blendOption=Linear}
    alphaScaleBiasClamp {bInterpResult=true}
    boneToModify {boneName=dyn_skirt_front_r_MSTR}
    rotation {pitch=5}
   }
   n56 = "Transform (Modify) Bone - Bone: dyn_skirt_front_c_MSTR"(ComponentPose=@n55, Alpha=@n53, Translation="0,0,0", Rotation="10,0,0", Scale="1,1,1") {
    rotationMode=BMM_Additive
    rotationSpace=BCS_BoneSpace
    scaleMode=BMM_Additive
    scaleSpace=BCS_BoneSpace
    translationMode=BMM_Additive
    translationSpace=BCS_BoneSpace
    alphaBoolBlend {blendOption=Linear}
    alphaScaleBiasClamp {bInterpResult=true}
    boneToModify {boneName=dyn_skirt_front_c_MSTR}
    rotation {pitch=10}
   }
   n57 = "Transform (Modify) Bone - Bone: dyn_skirtH_front_l_MSTR"(ComponentPose=@n56, Alpha=@n53, Translation="0,0,0", Rotation="7,0,0", Scale="1,1,1") {
    rotationMode=BMM_Additive
    rotationSpace=BCS_BoneSpace
    scaleMode=BMM_Additive
    scaleSpace=BCS_BoneSpace
    translationMode=BMM_Additive
    translationSpace=BCS_BoneSpace
    alphaBoolBlend {blendOption=Linear}
    alphaScaleBiasClamp {bInterpResult=true}
    boneToModify {boneName=dyn_skirtH_front_l_MSTR}
    rotation {pitch=7}
   }
   n58 = Get bIsSprinting()
   n59 = "Transform (Modify) Bone - Bone: dyn_skirt_front_c_MSTR"!(ComponentPose=@n57, bAlphaBoolEnabled:bool=@n58, Translation="0,0,0", Rotation="10,0,0", Scale="1,1,1") -> (Pose) {
    alphaInputType=Bool
    rotationMode=BMM_Additive
    rotationSpace=BCS_BoneSpace
    scaleMode=BMM_Additive
    scaleSpace=BCS_BoneSpace
    translationMode=BMM_Additive
    translationSpace=BCS_BoneSpace
    alphaBoolBlend {blendInTime=0.1, blendOption=Linear, blendOutTime=0.1}
    boneToModify {boneName=dyn_skirt_front_c_MSTR}
    rotation {pitch=10}
   }
   n60 = "Transform (Modify) Bone - Bone: dyn_skirtH_front_l_MSTR"!(ComponentPose=@n59, bAlphaBoolEnabled:bool=@n58, Translation="0,0,0", Rotation="5,-5,0", Scale="1,1,1") -> (Pose) {
    alphaInputType=Bool
    rotationMode=BMM_Additive
    rotationSpace=BCS_BoneSpace
    scaleMode=BMM_Additive
    scaleSpace=BCS_BoneSpace
    translationMode=BMM_Additive
    translationSpace=BCS_BoneSpace
    alphaBoolBlend {blendInTime=0.1, blendOption=Linear, blendOutTime=0.1}
    boneToModify {boneName=dyn_skirtH_front_l_MSTR}
    rotation {pitch=5, yaw=-5}
   }
   n61 = "Transform (Modify) Bone - Bone: dyn_skirt_front_r_MSTR"!(ComponentPose=@n60, bAlphaBoolEnabled:bool=@n58, Translation="0,0,0", Rotation="5,15,0", Scale="1,1,1") -> (Pose) {
    alphaInputType=Bool
    rotationMode=BMM_Additive
    rotationSpace=BCS_BoneSpace
    scaleMode=BMM_Additive
    scaleSpace=BCS_BoneSpace
    translationMode=BMM_Additive
    translationSpace=BCS_BoneSpace
    alphaBoolBlend {blendInTime=0.1, blendOption=Linear, blendOutTime=0.1}
    boneToModify {boneName=dyn_skirt_front_r_MSTR}
    rotation {pitch=5, yaw=15}
   }
   Output Pose(Result=Component To Local(ComponentPose=@n61)) {layerGroup=None}
   n62 = Control Rig!(
    Source=Copy Pose From Mesh() {bUseAttachedParent=true}
    Alpha
    IsCrouching:bool=OR Boolean(A=Get bIsCrouching(), B=Get bIsSkyDiving())
    IsBadEmotePlaying:bool=Get isBadEmotePlaying() -> (isBadEmotePlaying:bool)
    IsonGrindRail:bool=Get bIsOnGrindRail()
    jumping:double=0
    IsRBANEnabled:bool=Get bRigidBodyEnabled()
   ) -> (Pose) {
    alphaBoolBlend {blendOption=Linear}
    controlRigAssetReference {blueprintRigClass=&/M_MED_Ensemble_Maroon_coat_CtrlRig.M_MED_Ensemble_Maroon_coat_CtrlRig_C}
   }
   n63 = Get bIsBaseballSliding()
   n64 = "Transform (Modify) Bone - Bone: dyn_skirt_r_MSTR"!(
    ComponentPose=Local To Component(LocalPose=@n62)
    bAlphaBoolEnabled:bool=@n63
    Translation="-5.763,-0.000,2.687"
    Rotation="45,0,0"
    Scale="1,1,1"
   ) -> (Pose) {
    alphaInputType=Bool
    rotationMode=BMM_Additive
    rotationSpace=BCS_BoneSpace
    scaleMode=BMM_Additive
    scaleSpace=BCS_BoneSpace
    translationMode=BMM_Additive
    translationSpace=BCS_BoneSpace
    alphaBoolBlend {blendInTime=0.1, blendOption=Linear, blendOutTime=0.1}
    boneToModify {boneName=dyn_skirt_r_MSTR}
    rotation {pitch=45}
    translation {x=-5.763, z=2.687}
   }
   n65 = "Transform (Modify) Bone - Bone: dyn_skirt_bk_r_MSTR"!(ComponentPose=@n64, bAlphaBoolEnabled:bool=@n63, Translation="-5,0,0", Rotation="60,0,0", Scale="1,1,1") -> (Pose) {
    alphaInputType=Bool
    rotationMode=BMM_Additive
    rotationSpace=BCS_BoneSpace
    scaleMode=BMM_Additive
    scaleSpace=BCS_BoneSpace
    translationMode=BMM_Additive
    translationSpace=BCS_BoneSpace
    alphaBoolBlend {blendInTime=0.1, blendOption=Linear, blendOutTime=0.1}
    boneToModify {boneName=dyn_skirt_bk_r_MSTR}
    rotation {pitch=60}
    translation {x=-5}
   }
   n66 = "Transform (Modify) Bone - Bone: dyn_skirt_front_c_MSTR"(ComponentPose=@n65, bAlphaBoolEnabled=@n63, Translation="0,0,3", Rotation="0,0,0", Scale="1,1,1") {
    alphaInputType=Bool
    rotationMode=BMM_Additive
    rotationSpace=BCS_BoneSpace
    scaleMode=BMM_Additive
    scaleSpace=BCS_BoneSpace
    translationMode=BMM_Additive
    translationSpace=BCS_BoneSpace
    alphaBoolBlend {blendInTime=0.1, blendOption=Linear, blendOutTime=0.1}
    boneToModify {boneName=dyn_skirt_front_c_MSTR}
    translation {z=3}
   }
   n67 = "Transform (Modify) Bone - Bone: dyn_skirt_front_r_MSTR"(ComponentPose=@n66, bAlphaBoolEnabled=@n63, Translation="0,0,5", Rotation="0,0,0", Scale="1,1,1") {
    alphaInputType=Bool
    rotationMode=BMM_Additive
    rotationSpace=BCS_BoneSpace
    scaleMode=BMM_Additive
    scaleSpace=BCS_BoneSpace
    translationMode=BMM_Additive
    translationSpace=BCS_BoneSpace
    alphaBoolBlend {blendInTime=0.1, blendOption=Linear, blendOutTime=0.1}
    boneToModify {boneName=dyn_skirt_front_r_MSTR}
    translation {z=5}
   }
   n68 = "Transform (Modify) Bone - Bone: dyn_skirt_l_MSTR"!(ComponentPose=@n67, bAlphaBoolEnabled:bool=@n63, Translation="0,0,0", Rotation="35,0,0", Scale="1,1,1") -> (Pose) {
    alphaInputType=Bool
    rotationMode=BMM_Additive
    rotationSpace=BCS_BoneSpace
    scaleMode=BMM_Additive
    scaleSpace=BCS_BoneSpace
    translationMode=BMM_Additive
    translationSpace=BCS_BoneSpace
    alphaBoolBlend {blendInTime=0.1, blendOption=Linear, blendOutTime=0.1}
    boneToModify {boneName=dyn_skirt_l_MSTR}
    rotation {pitch=35}
   }
   n69 = "Transform (Modify) Bone - Bone: dyn_skirt_bk_l_MSTR"!(ComponentPose=@n68, bAlphaBoolEnabled:bool=@n63, Translation="0,0,0", Rotation="60.000,0.000,-0.000", Scale="1,1,1") -> (Pose) {
    alphaInputType=Bool
    rotationMode=BMM_Additive
    rotationSpace=BCS_BoneSpace
    scaleMode=BMM_Additive
    scaleSpace=BCS_BoneSpace
    translationMode=BMM_Additive
    translationSpace=BCS_BoneSpace
    alphaBoolBlend {blendInTime=0.1, blendOption=Linear, blendOutTime=0.1}
    boneToModify {boneName=dyn_skirt_bk_l_MSTR}
    rotation=<rot 59.9999 0 0>
   }
   n70 = Get bIsKneeSliding()
   n71 = "Transform (Modify) Bone - Bone: dyn_skirt_r_MSTR"(ComponentPose=@n69, bAlphaBoolEnabled=@n70, Translation="-12.983,-6.314,4.959", Rotation="30.000,54.736,35.264", Scale="1,1,1") {
    alphaInputType=Bool
    rotationMode=BMM_Additive
    rotationSpace=BCS_BoneSpace
    scaleMode=BMM_Additive
    scaleSpace=BCS_BoneSpace
    translationMode=BMM_Additive
    translationSpace=BCS_BoneSpace
    alphaBoolBlend {blendInTime=0.1, blendOption=Linear, blendOutTime=0.1}
    boneToModify {boneName=dyn_skirt_r_MSTR}
    rotation=<rot 30 54.736 35.264>
    translation=<-12.983 -6.314 4.959>
   }
   n72 = "Transform (Modify) Bone - Bone: dyn_skirt_bk_r_MSTR"(ComponentPose=@n71, bAlphaBoolEnabled=@n70, Translation="-15,0,-25", Rotation="90,0,0", Scale="1,1,1") {
    alphaInputType=Bool
    rotationMode=BMM_Additive
    rotationSpace=BCS_BoneSpace
    scaleMode=BMM_Additive
    scaleSpace=BCS_BoneSpace
    translationMode=BMM_Additive
    translationSpace=BCS_BoneSpace
    alphaBoolBlend {blendInTime=0.1, blendOption=Linear, blendOutTime=0.1}
    boneToModify {boneName=dyn_skirt_bk_r_MSTR}
    rotation {pitch=90}
    translation {x=-15, z=-25}
   }
   n73 = "Transform (Modify) Bone - Bone: dyn_skirt_l_MSTR"(ComponentPose=@n72, bAlphaBoolEnabled=@n70, Translation="-8.190,10.343,1.284", Rotation="16.664,-64.689,-31.231", Scale="1,1,1") {
    alphaInputType=Bool
    rotationMode=BMM_Additive
    rotationSpace=BCS_BoneSpace
    scaleMode=BMM_Additive
    scaleSpace=BCS_BoneSpace
    translationMode=BMM_Additive
    translationSpace=BCS_BoneSpace
    alphaBoolBlend {blendInTime=0.1, blendOption=Linear, blendOutTime=0.1}
    boneToModify {boneName=dyn_skirt_l_MSTR}
    rotation=<rot 16.664 -64.689 -31.231>
    translation=<-8.19 10.343 1.284>
   }
   n74 = "Transform (Modify) Bone - Bone: dyn_skirt_bk_l_MSTR"(ComponentPose=@n73, bAlphaBoolEnabled=@n70, Translation="-16.041,-0.000,-1.407", Rotation="85.000,-180.000,-180.000", Scale="1,1,1") {
    alphaInputType=Bool
    rotationMode=BMM_Additive
    rotationSpace=BCS_BoneSpace
    scaleMode=BMM_Additive
    scaleSpace=BCS_BoneSpace
    translationMode=BMM_Additive
    translationSpace=BCS_BoneSpace
    alphaBoolBlend {blendInTime=0.1, blendOption=Linear, blendOutTime=0.1}
    boneToModify {boneName=dyn_skirt_bk_l_MSTR}
    rotation=<rot 85.0001 -180 -180>
    translation {x=-16.041, z=-1.407}
   }
   Save cached pose 'SavedPose21276'(Pose=Component To Local(ComponentPose=@n74)) {cacheName=SavedPose21276}
   Save cached pose 'SavedPose16869'() {cacheName=SavedPose16869}
   Use cached pose 'SavedPose16869'()
   Use cached pose 'SavedPose16869'()
   Use cached pose 'SavedPose16869'()
   Get bIsCrouching()
   Get bIsOnGrindRail()
  ]}
 EventGraph {nodes [
   n1 = Event BlueprintUpdateAnimation()
   > Set Delta Time X(Delta Time X:double=@n1.DeltaTimeX) -> (Output_Get:double)
   > n2 = Sequence!(execute) -> (then_0, then_1, then_2:Exec, then_3:Exec, then_4:Exec, then_5:Exec, then_6:Exec)
   > n3 = Gravity Override State Function DEPRECATED(
    CharacterPartAnimInstance=Self-Reference()
    GameTimeInSeconds=Get Delta Time X() -> (Delta Time X:double)
    GravityOverrideParams=Get Gravity Override Params()
   )
   > Set Gravity Override(Gravity Override=@n3)
   Set Body Mesh(execute=@n2.then_1, Body Mesh:Skeletal Mesh Component Object Reference=GetOwnerSkeletalMeshComponent(self=Self-Reference())) -> (Output_Get:Skeletal Mesh Component Object Reference)
   > n4 = PlayerMovement(AnimInstance=Self-Reference(), XDotProductCoefficient=1, YDotProductCoefficient=1, ZDotProductCoefficient=1)
   > Set StrafeL(StrafeL:double=MapRangeClamped(Value=@n4, InRangeA=0, InRangeB=-1)) -> (Output_Get:double)
   > Set StrafeR(StrafeR:double=MapRangeClamped(Value=@n4, InRangeA=0, InRangeB=1)) -> (Output_Get:double)
   > Set Backward(Backward:double=MapRangeClamped(Value=@n4.YMovement, InRangeA=0, InRangeB=-1)) -> (Output_Get:double)
   > Set Forward(Forward:double=MapRangeClamped(Value=@n4.YMovement, InRangeA=0, InRangeB=1)) -> (Output_Get:double)
   n5 = Get bIsCrouching()
   Set Is Crouching(execute=@n2.then_2, Is Crouching:double="To Float (Boolean)"(InBool=@n5)) -> (Output_Get:double)
   > Set isNotCrouching(isNotCrouching:double="To Float (Boolean)"(InBool=NOT Boolean(A=@n5))) -> (Output_Get:double)
   n6 = Get bIsJumping()
   n7 = Get bIsFalling()
   > Set Is Jumping(Is Jumping:double="To Float (Boolean)"(InBool=OR Boolean(A=@n6, B=@n7))) -> (Output_Get:double)
   n8 = Get bIsSkyDiving()
   > Set Is Skydiving(Is Skydiving:double="To Float (Boolean)"(InBool=@n8)) -> (Output_Get:double)
   > Set Is Not Skydiving(Is Not Skydiving:double="To Float (Boolean)"(InBool=NOT Boolean(A=@n8))) -> (Output_Get:double)
   n9 = Get bIsInFrontEnd()
   > Set is Front End(is Front End:double="To Float (Boolean)"(InBool=@n9)) -> (Output_Get:double)
   > Set Is Not Front End(Is Not Front End:double="To Float (Boolean)"(InBool=NOT Boolean(A=@n9))) -> (Output_Get:double)
   n10 = GetOwnerSkeletalMeshComponent()
   n11 = BreakRotator(InRot=BreakTransform(InTransform=GetDeltaTransformFromRefPose(self=@n10, BoneName=thigh_r, BaseName=pelvis)).Rotation)
   Set thigh_rot_inr_r(execute=@n2.then_3, thigh_rot_inr_r:double=MapRangeClamped(Value=@n11.Pitch, InRangeA=0, InRangeB=-10, OutRangeA=1, OutRangeB=0)) -> (Output_Get:double)
   > Set thigh_rot_otr_r(thigh_rot_otr_r:double=MapRangeClamped(Value=@n11.Pitch, InRangeA=20, InRangeB=50)) -> (Output_Get:double)
   n12 = GetOwnerSkeletalMeshComponent()
   n13 = BreakRotator(InRot=BreakTransform(InTransform=GetDeltaTransformFromRefPose(self=@n12, BoneName=thigh_l, BaseName=pelvis)).Rotation)
   > Set thigh_rot_inr_l(thigh_rot_inr_l:double=MapRangeClamped(Value=@n13.Pitch, InRangeA=0, InRangeB=-10, OutRangeA=1, OutRangeB=0)) -> (Output_Get:double)
   > Set thigh_rot_otr_l(thigh_rot_otr_l:double=MapRangeClamped(Value=@n13.Pitch, InRangeA=20, InRangeB=50)) -> (Output_Get:double)
   > n14 = PlayerMovement(AnimInstance=Self-Reference(), XDotProductCoefficient=1, YDotProductCoefficient=1, ZDotProductCoefficient=1)
   > Set BwdMovement(BwdMovement:double=MapRangeClamped(Value=@n14.YMovement, InRangeA=0, InRangeB=-1)) -> (Output_Get:double)
   > Set FwdMovement(FwdMovement:double=MapRangeClamped(Value=@n14.YMovement, InRangeA=0, InRangeB=1)) -> (Output_Get:double)
   > Set RightMovement(RightMovement:double=MapRangeClamped(Value=@n14, InRangeA=0, InRangeB=1)) -> (Output_Get:double)
   > Set LeftMovement(LeftMovement:double=MapRangeClamped(Value=@n14, InRangeA=0, InRangeB=-1)) -> (Output_Get:double)
   n15 = BreakTransform(InTransform=GetDeltaTransformFromRefPose(self=GetOwnerSkeletalMeshComponent(), BoneName=calf_l, BaseName=thigh_l))
   > Set KneeLf(KneeLf:double=MapRangeClamped(Value=BreakRotator(InRot=@n15.Rotation).Yaw, InRangeA=0, InRangeB=100)) -> (Output_Get:double)
   > n16 = Branch(Condition=Get bIsSprinting())
   Set FwdMovement(execute=@n16.else, FwdMovement=0)
   n17 = GetAnimInstance(self=GetOwnerSkeletalMeshComponent())
   Is Valid(exec=@n2.then_4, InputObject=@n17)
   n18 = Montage_IsPlaying(self=@n17, Montage=/BRCosmetics/Animation/Game/MainPlayer/Emotes/SpeedRun/Emote_SpeedRun_CMM_M)
   > Set isCustomEmotePlaying(isCustomEmotePlaying:double=SelectFloat(A=1, B=0, bPickA=OR Boolean!(A=@n18, B=false, C:bool, D:bool, E:bool) -> (ReturnValue))) -> (Output_Get:double)
   n19 = GetAnimInstance(self=GetOwnerSkeletalMeshComponent())
   Is Valid(exec=@n2.then_5, InputObject=@n19)
   n20 = Montage_IsPlaying(self=@n19, Montage=/BRCosmetics/Animation/Game/MainPlayer/Emotes/Ebony/Emote_Ebony_CMM_M)
   > Set isBadEmotePlaying(isBadEmotePlaying:bool=@n20) -> (Output_Get:bool)
   Set KneeStraight(execute=@n2.then_6, KneeStraight:bool="Not Equal (Float)"(A=Get KneeLf() -> (KneeLf:double), B=1)) -> (Output_Get:bool)
   SelectFloat(A=1, B=0, bPickA=@n20)
  ]}
]
variables [
 Backward {type=number}
 Body Mesh {type=SkeletalMeshComponent}
 BwdMovement {type=number}
 Delta Time X {type=number}
 Forward {type=number}
 FwdMovement {type=number}
 Gravity Override {type=Vector}
 Gravity Override Params {
  type=GravityOverrideParamsStruct
  default {skydive_GoingUp {gravityOverride=<0 -400 -1200>}, skydive_Parachuting {gravityOverride=<0 0 -980>}}
 }
 Is Crouching {type=number}
 is Front End {type=number}
 Is Jumping {type=number}
 Is Not Front End {type=number}
 Is Not Skydiving {type=number}
 Is Skydiving {type=number}
 isBadEmotePlaying {type=boolean}
 isCustomEmotePlaying {type=number}
 isNotCrouching {type=number}
 KneeLf {type=number}
 KneeStraight {type=boolean}
 LeftMovement {type=number}
 RightMovement {type=number}
 StrafeL {type=number}
 StrafeR {type=number}
 thigh_rot_inr_l {type=number}
 thigh_rot_inr_r {type=number}
 thigh_rot_otr_l {type=number}
 thigh_rot_otr_r {type=number}
]
