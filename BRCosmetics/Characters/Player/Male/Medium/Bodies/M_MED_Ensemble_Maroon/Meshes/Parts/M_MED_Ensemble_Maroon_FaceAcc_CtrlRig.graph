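# Control Rig for the Maroon Ensemble male character's face accessory part, applying a corrective pose to two hair bones as the head moves.
# It reads the head's rotation relative to spine_05, takes the Euler Z component and remaps it to a clamped blend weight.
# That weight drives chained ModifyTransforms nodes that blend a fixed rotation and translation offset onto dyn_hair_1 and dyn_hair_2.
# The other dyn_hair bones in the hierarchy are untouched.
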ControlRigBlueprint /BRCosmetics/Characters/Player/Male/Medium/Bodies/M_MED_Ensemble_Maroon/Meshes/Parts/M_MED_Ensemble_Maroon_FaceAcc_CtrlRig
graphs [
 RigVMFunctionLibrary {links [], nodes []}
 RigVMModel {nodes [
   RigUnit_BeginExecution = &RigUnit_BeginExecution()
   n1 = &RigVMFunction_MathFloatRemap(
    SourceMaximum=-45
    SourceMinimum=10
    TargetMinimum=0
    bClamp=true
    Value=&RigVMFunction_MathQuaternionToEuler(
     RotationOrder=ZYX
     Value=&RigUnit_GetRelativeTransformForItem(Child=(Type=Bone,Name="head"), Parent=(Type=Bone,Name="spine_05")).RelativeTransform.Rotation
    ).Result.Z
   )
   n2 = &RigVMFunction_MathQuaternionFromEuler(Euler=(X=0,Y=0,Z=-65))
   &RigUnit_ModifyTransforms(
    ItemToModify=((Item=(Type=Bone,Name="dyn_hair_1"),Transform=(Rotation=(X=0,Y=0,Z=0,W=1),Translation=(X=-5,Y=-1,Z=0),Scale3D=(X=1,Y=1,Z=1))))
    Weight=@n1.Result
    ItemToModify.0.Transform.Rotation=@n2.Result
    ExecutePin=@RigUnit_BeginExecution.ExecutePin
   )
   > &RigUnit_ModifyTransforms(
    ItemToModify=((Item=(Type=Bone,Name="dyn_hair_2"),Transform=(Rotation=(X=0,Y=0,Z=0,W=1),Translation=(X=-5,Y=-1,Z=0),Scale3D=(X=1,Y=1,Z=1))))
    Weight=@n1.Result
    ItemToModify.0.Transform.Rotation=@n2.Result
   )
  ]}
]
hierarchy [
 root {loc=<0 0 0>, rot=<0 0 0 1>}
 pelvis {parent=root, loc=<0 -0.7071 93.1191>, rot=<0.0127 0.707 -0.0127 -0.707>}
 spine_01 {parent=pelvis, loc=<6.7494 -0.6388 0>, rot=<0 0 0.0503 -0.9987>}
 spine_02 {parent=spine_01, loc=<7.2493 -0.0308 0>, rot=<0 0 0.0041 -1>}
 spine_03 {parent=spine_02, loc=<6.0084 -0.0167 0>, rot=<0 0 0.0027 -1>}
 spine_04 {parent=spine_03, loc=<8.0999 -0.0224 0>, rot=<0 0 0.0026 -1>}
 spine_05 {parent=spine_04, loc=<10.0119 -0.0184 0>, rot=<0 0 -0.0016 -1>}
 neck_01 {parent=spine_05, loc=<11.3138 0.1529 0>, rot=<0 0 -0.2003 -0.9797>}
 neck_02 {parent=neck_01, loc=<4.5627 0 0>, rot=<0 0 0.0504 -0.9987>}
 head {parent=neck_02, loc=<5.3914 0 0>, rot=<0 0 0.115 -0.9934>}
 dyn_hair_1 {parent=head, loc=<2.6052 -6.4014 -4.104>, rot=<0.0755 -0.9771 -0.198 -0.0188>}
 dyn_hair_2 {parent=head, loc=<2.6052 -6.4014 3.3444>, rot=<0.0741 -0.9764 0.2014 0.0234>}
 dyn_hair_3 {parent=head, loc=<11.429 -6.1062 5.8477>, rot=<0.071 -0.8942 0.441 0.0293>}
 dyn_hair_4 {parent=head, loc=<11.4335 -6.6479 -0.1374>, rot=<0.0774 -0.9967 -0.0235 0.0116>}
 dyn_hair_5 {parent=head, loc=<11.4171 -4.6675 -6.4586>, rot=<0.0751 -0.7932 -0.6042 -0.016>}
 dyn_hair_6 {parent=head, loc=<11.3724 0.7423 -7.4704>, rot=<-0.0617 0.6934 0.7167 0.0421>}
 dyn_hair_7 {parent=head, loc=<11.3831 -0.5502 6.6399>, rot=<0.0566 -0.6291 0.7738 0.0476>}
 dyn_hair_8 {parent=head, loc=<11.3185 7.2674 4.5048>, rot=<0.0167 -0.2627 0.9623 0.0685>}
 dyn_hair_9 {parent=head, loc=<12.0006 7.07 -4.3669>, rot=<-0.0368 0.2194 0.973 0.0608>}
 dyn_hair_10 {parent=head, loc=<17.6848 4.0276 -2.8234>, rot=<0.1769 -0.1894 -0.7482 -0.6107>}
 dyn_hair_11 {parent=head, loc=<19.0264 0.3293 0.4303>, rot=<-0.4174 0.3904 -0.5683 -0.5919>}
 dyn_hair_12 {parent=head, loc=<16.2078 -1.4142 -2.8086>, rot=<0.5642 -0.63 -0.4001 -0.3532>}
 dyn_hair_13 {parent=head, loc=<18.1474 -3.6423 1.3893>, rot=<0.6717 -0.6915 0.2033 0.1711>}
 clavicle_l {parent=spine_05, loc=<4.3575 -0.1466 -2.3223>, rot=<0.0476 -0.7289 -0.0102 -0.6829>}
 clavicle_r {parent=spine_05, loc=<4.3575 -0.1466 2.3223>, rot=<0.7289 0.0476 -0.6829 0.0102>}
]
