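# Post-style Anim Blueprint for the Maroon ensemble's male medium face accessory mesh, adding secondary motion to its hair bones.
# It copies the attached parent's pose, runs the accessory's Control Rig, then chains single-body Anim Dynamics on dyn_hair_1–6 and 11–13.
# Bones 11–13 use larger boxes and mostly tighter angular limits.
# The dyn_hair_7–10 chain is disconnected from the output, and the event graph has no logic.
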AnimBlueprint /BRCosmetics/Characters/Player/Male/Medium/Bodies/M_MED_Ensemble_Maroon/Meshes/Parts/M_MED_Ensemble_Maroon_FaceAcc_AnimBP
target_skeleton=&/M_MED_Ensemble_Maroon_FaceAcc_Skeleton
graphs [
 AnimGraph {nodes [
   n1 = Anim Dynamics - Bone: dyn_hair_7(Alpha=1, ExternalForce="0,0,0") {
    alphaBoolBlend {blendOption=Linear}
    boundBone {boneName=dyn_hair_7}
    physicsBodyDefinitions [{
      boundBone {boneName=dyn_hair_7}
      boxExtents=<20 8 16>
      constraintSetup {angularLimitsMax=<0 2 2>, angularLimitsMin=<0 -3 -20>, angularTarget=<0 0 0>}
      localJointOffset=<8 0 0>
     }]
   }
   n2 = Anim Dynamics - Bone: dyn_hair_8(ComponentPose=@n1, Alpha=1, ExternalForce="0,0,0") {
    alphaBoolBlend {blendOption=Linear}
    boundBone {boneName=dyn_hair_8}
    physicsBodyDefinitions [{
      boundBone {boneName=dyn_hair_8}
      boxExtents=<20 8 16>
      constraintSetup {angularLimitsMax=<0 2 2>, angularLimitsMin=<0 -3 -20>, angularTarget=<0 0 0>}
      localJointOffset=<8 0 0>
     }]
   }
   n3 = Anim Dynamics - Bone: dyn_hair_9(ComponentPose=@n2, Alpha=1, ExternalForce="0,0,0") {
    alphaBoolBlend {blendOption=Linear}
    boundBone {boneName=dyn_hair_9}
    physicsBodyDefinitions [{
      boundBone {boneName=dyn_hair_9}
      boxExtents=<20 8 16>
      constraintSetup {angularLimitsMax=<0 2 2>, angularLimitsMin=<0 -3 -20>, angularTarget=<0 0 0>}
      localJointOffset=<8 0 0>
     }]
   }
   Anim Dynamics - Bone: dyn_hair_10(ComponentPose=@n3, Alpha=1, ExternalForce="0,0,0") {
    alphaBoolBlend {blendOption=Linear}
    boundBone {boneName=dyn_hair_10}
    physicsBodyDefinitions [{
      boundBone {boneName=dyn_hair_10}
      boxExtents=<20 8 16>
      constraintSetup {angularLimitsMax=<0 2 2>, angularLimitsMin=<0 -3 -20>, angularTarget=<0 0 0>}
      localJointOffset=<8 0 0>
     }]
   }
   n4 = Control Rig(Source=Copy Pose From Mesh() {bUseAttachedParent=true}) {
    alphaBoolBlend {blendOption=Linear}
    controlRigAssetReference {blueprintRigClass=&/M_MED_Ensemble_Maroon_FaceAcc_CtrlRig.M_MED_Ensemble_Maroon_FaceAcc_CtrlRig_C}
   }
   n5 = Anim Dynamics - Bone: dyn_hair_1(ComponentPose=Local To Component(LocalPose=@n4), Alpha=1, ExternalForce="0,0,0") {
    alphaBoolBlend {blendOption=Linear}
    boundBone {boneName=dyn_hair_1}
    physicsBodyDefinitions [{
      boundBone {boneName=dyn_hair_1}
      boxExtents=<20 8 16>
      constraintSetup {angularLimitsMax=<0 2 2>, angularLimitsMin=<0 -3 -20>, angularTarget=<0 0 0>}
      localJointOffset=<8 0 0>
     }]
   }
   n6 = Anim Dynamics - Bone: dyn_hair_2(ComponentPose=@n5, Alpha=1, ExternalForce="0,0,0") {
    alphaBoolBlend {blendOption=Linear}
    boundBone {boneName=dyn_hair_2}
    physicsBodyDefinitions [{
      boundBone {boneName=dyn_hair_2}
      boxExtents=<20 8 16>
      constraintSetup {angularLimitsMax=<0 2 2>, angularLimitsMin=<0 -3 -20>, angularTarget=<0 0 0>}
      localJointOffset=<8 0 0>
     }]
   }
   n7 = Anim Dynamics - Bone: dyn_hair_3(ComponentPose=@n6, Alpha=1, ExternalForce="0,0,0") {
    alphaBoolBlend {blendOption=Linear}
    boundBone {boneName=dyn_hair_3}
    physicsBodyDefinitions [{
      boundBone {boneName=dyn_hair_3}
      boxExtents=<20 8 16>
      constraintSetup {angularLimitsMax=<0 2 2>, angularLimitsMin=<0 -3 -20>, angularTarget=<0 0 0>}
      localJointOffset=<8 0 0>
     }]
   }
   n8 = Anim Dynamics - Bone: dyn_hair_4(ComponentPose=@n7, Alpha=1, ExternalForce="0,0,0") {
    alphaBoolBlend {blendOption=Linear}
    boundBone {boneName=dyn_hair_4}
    physicsBodyDefinitions [{
      boundBone {boneName=dyn_hair_4}
      boxExtents=<20 8 16>
      constraintSetup {angularLimitsMax=<0 2 2>, angularLimitsMin=<0 -3 -20>, angularTarget=<0 0 0>}
      localJointOffset=<8 0 0>
     }]
   }
   n9 = Anim Dynamics - Bone: dyn_hair_5(ComponentPose=@n8, Alpha=1, ExternalForce="0,0,0") {
    alphaBoolBlend {blendOption=Linear}
    boundBone {boneName=dyn_hair_5}
    physicsBodyDefinitions [{
      boundBone {boneName=dyn_hair_5}
      boxExtents=<20 8 16>
      constraintSetup {angularLimitsMax=<0 2 2>, angularLimitsMin=<0 -3 -20>, angularTarget=<0 0 0>}
      localJointOffset=<8 0 0>
     }]
   }
   n10 = Anim Dynamics - Bone: dyn_hair_6(ComponentPose=@n9, Alpha=1, ExternalForce="0,0,0") {
    alphaBoolBlend {blendOption=Linear}
    boundBone {boneName=dyn_hair_6}
    physicsBodyDefinitions [{
      boundBone {boneName=dyn_hair_6}
      boxExtents=<20 8 16>
      constraintSetup {angularLimitsMax=<0 2 2>, angularLimitsMin=<0 -3 -20>, angularTarget=<0 0 0>}
      localJointOffset=<8 0 0>
     }]
   }
   n11 = Anim Dynamics - Bone: dyn_hair_11(ComponentPose=@n10, Alpha=1, ExternalForce="0,0,0") {
    alphaBoolBlend {blendOption=Linear}
    boundBone {boneName=dyn_hair_11}
    physicsBodyDefinitions [{
      boundBone {boneName=dyn_hair_11}
      boxExtents=<30 10 24>
      constraintSetup {angularLimitsMax=<0 1 5>, angularLimitsMin=<0 -2 -10>, angularTarget=<0 0 0>}
      localJointOffset=<0.1 -0.5 0>
     }]
   }
   n12 = Anim Dynamics - Bone: dyn_hair_12(ComponentPose=@n11, Alpha=1, ExternalForce="0,0,0") {
    alphaBoolBlend {blendOption=Linear}
    boundBone {boneName=dyn_hair_12}
    physicsBodyDefinitions [{
      boundBone {boneName=dyn_hair_12}
      boxExtents=<30 10 24>
      constraintSetup {angularLimitsMax=<0 1 5>, angularLimitsMin=<0 -2 -10>, angularTarget=<0 0 0>}
      localJointOffset=<0.1 -0.5 0>
     }]
   }
   n13 = Anim Dynamics - Bone: dyn_hair_13(ComponentPose=@n12, Alpha=1, ExternalForce="0,0,0") {
    alphaBoolBlend {blendOption=Linear}
    boundBone {boneName=dyn_hair_13}
    physicsBodyDefinitions [{
      boundBone {boneName=dyn_hair_13}
      boxExtents=<30 10 24>
      constraintSetup {angularLimitsMax=<0 1 5>, angularLimitsMin=<0 -2 -10>, angularTarget=<0 0 0>}
      localJointOffset=<0.1 -0.5 0>
     }]
   }
   Output Pose(Result=Component To Local(ComponentPose=@n13)) {layerGroup=None}
  ]}
 EventGraph {nodes [Event BlueprintUpdateAnimation()]}
]
variables []
